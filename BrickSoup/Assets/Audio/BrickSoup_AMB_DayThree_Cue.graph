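# BrickSoup Day Three ambience cue, layering two looping beds with intermittent ice-lake spot sounds for an outdoor ambient soundscape.
# A looping, randomly delayed branch picks one of seven Snow Ice Lake waves, varies pitch and volume, and uses wind one-shot attenuation.
# A mixer combines it with the louder-weighted Day2 and Day3 loops, both using close ambience attenuation.
# The cue also sends to the distant reverb submix.
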SoundCue /BrickSoup/Assets/Audio/BrickSoup_AMB_DayThree_Cue
root=Mixer_0
nodes [
 {id=Attenuation_0, children [Random_0], properties {attenuationSettings=/Game/Sounds/Attenuation/Fort_TileAmbient_OneShot_Wind_Att}}
 {id=Attenuation_1, children [WavePlayer_0], properties {attenuationSettings=&/ATT_BrickSoup_Close_AMB}}
 {id=Attenuation_2, children [WavePlayer_6], properties {attenuationSettings=&ATT_BrickSoup_Close_AMB}}
 {id=Delay_0, children [Modulator_0], properties {delayMax=3, delayMin=1}}
 {id=Looping_0, children [Delay_0]}
 {id=Mixer_0, children [Looping_0, Attenuation_1, Attenuation_2], properties {inputVolume [1,4,3]}}
 {id=Modulator_0, children [Attenuation_0], properties {pitchMax=1.1, pitchMin=0.9, volumeMin=0.85}}
 {
  id=Random_0
  children [WavePlayer_1, WavePlayer_2, WavePlayer_3, WavePlayer_4, WavePlayer_5, WavePlayer_10, WavePlayer_11]
  properties {weights [1,1,1,1,1,1,1]}
 }
 {id=WavePlayer_0, properties {bLooping=true, soundWaveAssetPtr=&/S33_BrickSoup_Day2_Loop}}
 {id=WavePlayer_1, properties {soundWaveAssetPtr=/Game/Sounds/Asteria/Biomes/RealityWarrior/Spots/Spot_Snow_Ice_Lake_07}}
 {id=WavePlayer_10, properties {soundWaveAssetPtr=/Game/Sounds/Asteria/Biomes/RealityWarrior/Spots/Spot_Snow_Ice_Lake_05}}
 {id=WavePlayer_11, properties {soundWaveAssetPtr=/Game/Sounds/Asteria/Biomes/RealityWarrior/Spots/Spot_Snow_Ice_Lake_06}}
 {id=WavePlayer_2, properties {soundWaveAssetPtr=/Game/Sounds/Asteria/Biomes/RealityWarrior/Spots/Spot_Snow_Ice_Lake_01}}
 {id=WavePlayer_3, properties {soundWaveAssetPtr=/Game/Sounds/Asteria/Biomes/RealityWarrior/Spots/Spot_Snow_Ice_Lake_02}}
 {id=WavePlayer_4, properties {soundWaveAssetPtr=/Game/Sounds/Asteria/Biomes/RealityWarrior/Spots/Spot_Snow_Ice_Lake_03}}
 {id=WavePlayer_5, properties {soundWaveAssetPtr=/Game/Sounds/Asteria/Biomes/RealityWarrior/Spots/Spot_Snow_Ice_Lake_04}}
 {id=WavePlayer_6, properties {bLooping=true, soundWaveAssetPtr=&/S33_BrickSoup_Day3_Loop}}
]
properties {
 maxDistance=8000
 soundClassObject=&Fort_Ambient_OneShots_Outdoors
 volumeMultiplier=0.5
 preEffectBusSends [{}]
 soundSubmixSends [{minSendLevel=1, soundSubmix=&Fortnite_Reverb_Distant}]
}
