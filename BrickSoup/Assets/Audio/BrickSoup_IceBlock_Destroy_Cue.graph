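# Sound cue for an ice block being destroyed in BrickSoup, built from treasure chest ice explosion waves.
# A random node picks one of six equally weighted explosion variants and feeds it through a modulator at the root for variation.
# It plays in the impacts sound class with small melee weapon attenuation and references a shared object-impact concurrency set, alongside override values that stop the oldest sound.
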SoundCue /BrickSoup/Assets/Audio/BrickSoup_IceBlock_Destroy_Cue
root=Modulator_37
nodes [
 {id=Modulator_37, children [Random_0]}
 {
  id=Random_0
  children [WavePlayer_0, WavePlayer_1, WavePlayer_2, WavePlayer_3, WavePlayer_11, WavePlayer_12]
  properties {weights [1,1,1,1,1,1]}
 }
 {id=WavePlayer_0, properties {soundWaveAssetPtr=&/S33_BrickSoup_TreasureChest_IceExplode_01}}
 {id=WavePlayer_1, properties {soundWaveAssetPtr=&/S33_BrickSoup_TreasureChest_IceExplode_02}}
 {id=WavePlayer_11, properties {soundWaveAssetPtr=&/S33_BrickSoup_TreasureChest_IceExplode_05}}
 {id=WavePlayer_12, properties {soundWaveAssetPtr=&/S33_BrickSoup_TreasureChest_IceExplode_06}}
 {id=WavePlayer_2, properties {soundWaveAssetPtr=&/S33_BrickSoup_TreasureChest_IceExplode_03}}
 {id=WavePlayer_3, properties {soundWaveAssetPtr=&/S33_BrickSoup_TreasureChest_IceExplode_04}}
]
properties {
 attenuationSettings=&Fort_Weapon_Melee_Sml_Att
 maxDistance=2800
 soundClassObject=&Fort_Impacts
 volumeMultiplier=0.96
 concurrencyOverrides {maxCount=3, resolutionRule=StopOldest}
 concurrencySet [&CONC_Object_Impacts_FourMax]
}
