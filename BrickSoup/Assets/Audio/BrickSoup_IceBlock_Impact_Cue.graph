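# Sound cue for ice block impacts in BrickSoup, built from the shared BrickSoup melee impact recordings.
# A Random node picks one of eight equally weighted wave players, and the cue raises the volume above default.
# It routes to the Fort impacts sound class and references small melee weapon attenuation settings.
# It also carries override values for logarithmic low-pass falloff and an owner-limited, two-voice concurrency that stops the quietest.
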SoundCue /BrickSoup/Assets/Audio/BrickSoup_IceBlock_Impact_Cue
root=Random_0
nodes [
 {
  id=Random_0
  children [WavePlayer_2, WavePlayer_3, WavePlayer_14, WavePlayer_15, WavePlayer_16, WavePlayer_17, WavePlayer_18, WavePlayer_19]
  properties {weights [1,1,1,1,1,1,1,1]}
 }
 {id=WavePlayer_14, properties {soundWaveAssetPtr=&/S33_BrickSoup_Melee_Impact_07}}
 {id=WavePlayer_15, properties {soundWaveAssetPtr=&/S33_BrickSoup_Melee_Impact_05}}
 {id=WavePlayer_16, properties {soundWaveAssetPtr=&/S33_BrickSoup_Melee_Impact_04}}
 {id=WavePlayer_17, properties {soundWaveAssetPtr=&/S33_BrickSoup_Melee_Impact_03}}
 {id=WavePlayer_18, properties {soundWaveAssetPtr=&/S33_BrickSoup_Melee_Impact_02}}
 {id=WavePlayer_19, properties {soundWaveAssetPtr=&/S33_BrickSoup_Melee_Impact_01}}
 {id=WavePlayer_2, properties {soundWaveAssetPtr=&/S33_BrickSoup_Melee_Impact_10}}
 {id=WavePlayer_3, properties {soundWaveAssetPtr=&/S33_BrickSoup_Melee_Impact_09}}
]
properties {
 attenuationSettings=&Fort_Weapon_Melee_Sml_Att
 maxDistance=2800
 soundClassObject=&Fort_Impacts
 volumeMultiplier=1.8
 attenuationOverrides {
  bAttenuateWithLPF=true
  distanceAlgorithm=Logarithmic
  falloffDistance=800
  lPFRadiusMax=800
  lPFRadiusMin=400
  attenuationShapeExtents {x=300}
 }
 concurrencyOverrides {bLimitToOwner=true, maxCount=2, resolutionRule=StopQuietest}
}
