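# Translucent unlit aurora borealis material for Niagara mesh particles and ribbons, with beat-reactive bumps.
# Skewed, panning noise layers form rays that combine with a bottom glow and edge fades, tinted by a three-color vertical gradient.
# Dynamic parameters BeatBump and BeatOffset drive a dual-sine mask for brightness and upward offset, added to sine undulation along normals.
# Two-sided Fresnel, particle alpha and feature-level-switched depth fade shape opacity.
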Material /BrickSoup/Assets/FX/M_BrickSoup_AuroraBorealis
graph {
 properties {
  bCastRayTracedShadows=false
  bUsedWithNiagaraMeshParticles=true
  bUsedWithNiagaraRibbons=true
  blendMode=BLEND_Translucent
  pixelDepthOffsetMode=PDOM_Legacy
  shadingModel=MSM_Unlit
  usedShadingModels=Unlit
 }
 nodes [
  n1 = OneMinus(None=ScalarParameter() {defaultValue=0.4, group=02 - Color, parameterName=BottomColorTreshold})
  n2 = Add(A=@n1, B=ScalarParameter() {defaultValue=0.5, group=02 - Color, parameterName=BottomColorSoftness})
  n3 = Add(
   A=ComponentMask(None=TextureCoordinate()) {g=true}
   B=Multiply(A=NamedRerouteUsage(), B=ScalarParameter() {defaultValue=0.2, group=02 - Color, parameterName=ColorDistoStregth})
  )
  n4 = OneMinus(None=ScalarParameter() {defaultValue=1.55, group=02 - Color, parameterName=TopColorTreshold})
  n5 = ScalarParameter() {defaultValue=2, group=02 - Color, parameterName=TopColorSoftness}
  n6 = LinearInterpolate(
   A=VectorParameter() {group=02 - Color, parameterName=TopColor, defaultValue=<rgba 0.8338 0 0.9896 1>}.RGB
   B=LinearInterpolate(
    A=VectorParameter() {group=02 - Color, parameterName=MidColor, defaultValue=<rgba 0.3277 0.8596 1 1>}.RGB
    B=VectorParameter() {group=02 - Color, parameterName=BottomColor, defaultValue=<rgba 0.899 3 2.2034 1>}.RGB
    Alpha=SmoothStep(Min=@n1, Max=@n2, Value=@n3)
   )
   Alpha=SmoothStep(Min=@n4, Max=Add(A=@n4, B=@n5), Value=@n3)
  )
  n7 = TextureCoordinate()
  n8 = ComponentMask(None=@n7) {g=true}
  n9 = ScalarParameter() {defaultValue=0.005, group=00 - BaseSettings, parameterName=Skew Amount}
  n10 = Constant() {r=-0.5}
  n11 = Multiply(A=Add(A=Multiply(A=@n8, B=@n9), B=@n9), B=@n10)
  n12 = OneMinus(None=@n8)
  n13 = Multiply(A=Add(A=Multiply(A=@n12, B=@n9), B=@n9), B=@n10)
  n14 = Add(A=@n7, B=AppendVector(A=Subtract(A=@n11, B=@n13), B=Constant()))
  n15 = Multiply(A=@n14, B=AppendVector(A=ScalarParameter() {defaultValue=20, group=01-Opacity, parameterName=Mask1 Tile X}, B=Constant() {r=0.1}))
  n16 = AppendVector(A=Constant(), B=Constant())
  n17 = Constant() {r=7}
  n18 = Multiply(A=Time(), B=Constant() {r=0.5})
  n19 = ScalarParameter() {defaultValue=0.02, group=01-Opacity, parameterName=Rand Opacity Speed X}
  n20 = ScalarParameter() {defaultValue=0.02, group=01-Opacity, parameterName=Rand Opacity Speed Y}
  n21 = TextureObject() {samplerType=SAMPLERTYPE_LinearColor, texture=&HighContrastNoise}
  n22 = TextureSample(UVs=Panner(Coordinate=Add(A=@n15, B=@n16), Time=Multiply(A=@n18, B=@n17), Speed=AppendVector(A=@n19, B=@n20)), Tex=@n21)
  n23 = ScalarParameter() {defaultValue=35, group=01-Opacity, parameterName=Mask2 Tile X}
  n24 = Constant() {r=0.1}
  n25 = Add(A=Multiply(A=@n14, B=AppendVector(A=@n23, B=@n24)), B=AppendVector(A=Constant() {r=0.5432}, B=Constant() {r=0.2543}))
  n26 = Multiply(A=@n18, B=Constant() {r=8})
  n27 = AppendVector(
   A=ScalarParameter() {defaultValue=0.02, group=01-Opacity, parameterName=Rand Opacity Speed X}
   B=ScalarParameter() {defaultValue=0.02, group=01-Opacity, parameterName=Rand Opacity Speed Y}
  )
  n28 = Power(
   Base=TextureSample(UVs=Panner(Coordinate=@n25, Time=@n26, Speed=@n27), Tex=@n21) {samplerType=SAMPLERTYPE_LinearColor, texture=&HighContrastNoise}.R
   Exp=ScalarParameter() {defaultValue=1, group=01-Opacity, parameterName=Random Opacity Pow}
  )
  n29 = ScalarParameter() {defaultValue=0.5, group=01-Opacity, parameterName=Random Opacity Secondary Pow}
  n30 = Max(A=Multiply(A=@n22.G, B=@n28), B=Multiply(A=@n22.B, B=Power(Base=@n28, Exp=@n29)))
  n31 = Add(
   A=TextureCoordinate()
   B=Multiply(A=NamedRerouteUsage(), B=ScalarParameter() {defaultValue=0.3, group=01-Opacity, parameterName=BottomOpacityDistortion}) {constB=0.1}
  )
  n32 = Power(Base=SmoothStep(Value=ComponentMask(None=@n31) {g=true}) {constMax=0.8, constMin=0.4}) {constExponent=1}
  n33 = ScalarParameter() {defaultValue=0.4, group=01-Opacity, parameterName=BottomOpacityStrength}
  n34 = Multiply(
   A=Add(A=NamedRerouteDeclaration(None=@n30) {name=RaysAlpha, nodeColor=<rgba 0 1 0.9531 1>}, B=Multiply(A=@n32, B=@n33) {constB=0.25})
   B=NamedRerouteUsage()
  )
  n35 = Multiply(
   A=ScalarParameter() {defaultValue=2, group=02 - Color, parameterName=Emission Mult}
   B=Desaturation(None=Multiply(A=@n6, B=@n34), Fraction=Constant() {r=-5})
  )
  n36 = LinearInterpolate(B=ScalarParameter() {defaultValue=5, parameterName=BeatBumpColorMult}, Alpha=NamedRerouteUsage()) {constA=1}
  MaterialFunctionCall(Emissive=Multiply(A=VertexColor(), B=Multiply(A=@n35, B=@n36))) {materialFunction=&MF_FX_EmissiveAdjustments}
  n37 = ScalarParameter() {group=01-Opacity, parameterName=SideFadeStart, sortPriority=65}
  n38 = Add(A=@n37, B=ScalarParameter() {defaultValue=0.04, group=01-Opacity, parameterName=SideFadeSmoothness, sortPriority=65})
  n39 = Add(A=TextureCoordinate(), B=Multiply(A=@n22.G, B=Constant2Vector() {g=0.05}) {constB=0.05}) {constB=0}
  n40 = ComponentMask(None=@n39) {r=true}
  n41 = Min(A=@n40, B=OneMinus(None=@n40))
  n42 = ScalarParameter() {defaultValue=-0.2, group=01-Opacity, parameterName=TopFadeStart, sortPriority=65}
  n43 = ComponentMask(None=@n39) {g=true}
  n44 = SmoothStep(
   Min=@n42
   Max=Add(A=@n42, B=ScalarParameter() {defaultValue=0.7, group=01-Opacity, parameterName=TopFadeSmoothness, sortPriority=65})
   Value=@n43
  ) {constMax=0.3}
  n45 = ScalarParameter() {group=01-Opacity, parameterName=BottomFadeStart, sortPriority=65}
  n46 = SmoothStep(
   Min=@n45
   Max=Add(A=@n45, B=ScalarParameter() {defaultValue=0.15, group=01-Opacity, parameterName=BottomFadeSmoothness, sortPriority=65})
   Value=OneMinus(None=@n43)
  ) {constMax=0.1}
  NamedRerouteDeclaration(None=Multiply(A=SmoothStep(Min=@n37, Max=@n38, Value=@n41) {constMax=0.5}, B=Min(A=@n44, B=@n46))) {name=EdgeMask, nodeColor=<rgba 0.8726 1 0.1758 1>}
  n47 = DynamicParameter() {defaultValue=<rgba 0 1 1 1>, paramNames [BeatBump,BeatOffset,Param3,Param4]}
  n48 = ScalarParameter() {defaultValue=10, group=04 - BeatBump, parameterName=BeatBumScale}
  n49 = TextureCoordinate()
  n50 = ComponentMask(None=Add(A=Multiply(A=@n49, B=@n48), B=@n47.BeatOffset)) {r=true}
  n51 = Saturate(None=ConstantBiasScale(None=Sine(None=@n50)))
  n52 = Multiply(A=@n48) {constB=0.5}
  n53 = ComponentMask(None=Add(A=Multiply(A=@n47.BeatOffset) {constB=0.75}, B=Multiply(A=@n49, B=@n52))) {r=true}
  n54 = Saturate(None=ConstantBiasScale(None=Sine(None=@n53)))
  NamedRerouteDeclaration(None=Multiply(A=@n47.BeatBump, B=Multiply(A=@n51, B=@n54))) {name=BeatBumpAlpha, nodeColor=<rgba 1 0 0.1875 1>}
  n55 = Time()
  n56 = ScalarParameter() {defaultValue=0.001, group=03 - Ondulation, parameterName=OndulationSpeed}
  n57 = Multiply(
   A=Add(A=Multiply(A=@n55, B=@n56), B=TextureCoordinate())
   B=ScalarParameter() {defaultValue=22, group=03 - Ondulation, parameterName=OndulationFreq}
  )
  n58 = Multiply(
   A=Sine(None=ComponentMask(None=@n57) {r=true})
   B=ScalarParameter() {defaultValue=50, group=03 - Ondulation, parameterName=OndulationStrength}
  )
  n59 = VertexNormalWS()
  n60 = Multiply(A=NamedRerouteUsage(), B=ScalarParameter() {defaultValue=75, group=04 - BeatBump, parameterName=BeatBumpWPOStrength})
  n61 = Constant3Vector() {constant=<rgba 0 0 1 0>}
  Reroute(None=Add(A=Multiply(A=@n58, B=@n59), B=Multiply(A=@n60, B=@n61))) {desc=WPO}
  n62 = ParticleColor()
  n63 = Saturate(None=DotProduct(A=CameraVectorWS(), B=VertexNormalWS()))
  n64 = CameraVectorWS()
  n65 = VertexNormalWS()
  n66 = Power(
   Base=Abs(None=DotProduct(A=@n64, B=@n65)) {desc=for 2 sided}
   Exp=ScalarParameter() {defaultValue=4, group=01-Opacity, parameterName=FresnelSharpness}
  )
  n67 = ScalarParameter() {defaultValue=1, group=01-Opacity, parameterName=Overall Alpha Mult}
  n68 = Multiply(
   A=Saturate(None=Multiply(A=@n34, B=@n67)) {desc=opacity}
   B=Multiply(A=Power(Base=@n63) {constExponent=0.5}, B=Saturate(None=@n66)) {constA=1}
  ) {desc=texture alpha}
  n69 = DepthFade(FadeDistance=ScalarParameter() {defaultValue=150, group=01-Opacity, parameterName=DepthFadeDist})
  n70 = Constant() {r=1}
  Reroute(None=Multiply(A=Multiply(A=@n62.A, B=@n68), B=FeatureLevelSwitch(Default=@n69, ES2=@n70, ES3_1=@n70, SM5=@n69, SM6=@n69))) {desc=opacity angle fade}
 ]
}
