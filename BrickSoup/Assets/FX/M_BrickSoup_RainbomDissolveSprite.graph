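# Translucent unlit Niagara material for sprite, ribbon and mesh-particle effects that transitions noise-driven three-tone color and a rainbow gradient.
# Panning noise, offset by dynamic parameters and vertex alpha, drives color and distorts UVs, which can be swizzled or radial.
# Opacity combines distorted edge fades, a particle-driven swipe mask and depth fade, which is skipped on ES2/ES3.1.
# An optional static switch adds spline thickening through world position offset.
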Material /BrickSoup/Assets/FX/M_BrickSoup_RainbomDissolveSprite
graph {
 properties {
  bAutomaticallySetUsageInEditor=false
  bUsedWithNiagaraMeshParticles=true
  bUsedWithNiagaraRibbons=true
  bUsedWithNiagaraSprites=true
  blendMode=BLEND_Translucent
  shadingModel=MSM_Unlit
  usedShadingModels=Unlit
 }
 nodes [
  n1 = ScalarParameter() {defaultValue=0.5, group=03- Color, parameterName=FirstColorThreshold}
  n2 = Add(A=@n1, B=ScalarParameter() {defaultValue=0.25, group=03- Color, parameterName=FirstColorSoftness})
  n3 = VectorParameter() {group=03- Color, parameterName=ColorTex_Settings, channelNames {a=PanY, b=PanX, g=ScaleY, r=ScaleX}, defaultValue=<rgba 1 0.5 -0.7 0>}
  n4 = Panner(Coordinate=NamedRerouteUsage(), Speed=AppendVector(A=@n3.PanX, B=@n3.PanX)) {speedX=-0.5}
  n5 = AppendVector(A=@n3.ScaleX, B=@n3.ScaleX)
  n6 = Add(
   A=DynamicParameter() {defaultValue=<rgba 0.5 0.5 0 1>, paramNames [UV1 U Offset, Swipe, ColorTransition, LinearMask]}.UV1 U Offset
   B=VertexColor().A
  )
  n7 = Constant()
  n8 = Add(
   A=Add(A=Multiply(A=@n4, B=@n5), B=AppendVector(A=@n6, B=@n7))
   B=Multiply(A=NamedRerouteUsage(), B=ScalarParameter() {defaultValue=0.2, group=03- Color, parameterName=ColorDistoStrength})
  )
  n9 = TextureSampleParameter2D(UVs=@n8) {
   group=03- Color
   parameterName=ColorTex
   samplerType=SAMPLERTYPE_LinearColor
   texture=/Game/Effects/Fort_Effects/Gameplay/SmokeDustExplosions/Textures/T_NoisePack_LinearColor
  }
  n10 = TextureSampleParameter2D(UVs=@n8) {
   group=03- Color
   parameterName=AltColorTex
   samplerType=SAMPLERTYPE_LinearGrayscale
   texture=/Game/Effects/Fort_Effects/Textures/Tile/T_CellNoise_Tile_M
  }
  n11 = ComponentMask(None=StaticSwitchParameter(True=@n9.RGB, False=@n10.RGB) {defaultValue=true, desc=FORT-778441, group=03- Color, parameterName=ColorTextureIsLinear}) {r=true}
  n12 = Saturate(None=@n11)
  n13 = ScalarParameter() {defaultValue=0.4, group=03- Color, parameterName=SecondColorThreshold}
  n14 = ScalarParameter() {defaultValue=0.25, group=03- Color, parameterName=SecondColorSoftness}
  n15 = LinearInterpolate(
   A=VectorParameter() {group=03- Color, parameterName=OuterColorMult, sortPriority=0, defaultValue=<rgba 0.7 0.3495 0 1>}.RGB
   B=LinearInterpolate(
    A=VectorParameter() {group=03- Color, parameterName=MidColorMult, sortPriority=0, defaultValue=<rgba 0.8 0.4553 0 1>}.RGB
    B=VectorParameter() {group=03- Color, parameterName=InnerColorMult, sortPriority=0, defaultValue=<rgba 1 0.6702 0 1>}.RGB
    Alpha=SmoothStep(Min=@n1, Max=@n2, Value=@n12) {constMin=0.5}
   )
   Alpha=SmoothStep(Min=@n13, Max=Add(A=@n13, B=@n14), Value=Saturate(None=@n11)) {constMin=0.5}
  )
  n16 = NamedRerouteUsage()
  n17 = ScalarParameter() {group=03- Color, parameterName=RainbowDistoStrength}
  n18 = TextureSample(UVs=Add(A=NamedRerouteUsage(), B=Multiply(A=@n16, B=@n17))) {texture=/Game/Gadgets/Assets/VinderTech_GliderChute/Glider_Peach_Male/FX/T_Glider_Peach_rainbow_gradient}
  n19 = Add(A=@n11)
  n20 = DynamicParameter() {defaultValue=<rgba 0.5 0.5 0 1>, paramNames [UV1 U Offset, Swipe, ColorTransition, LinearMask]}
  n21 = Multiply(A=@n20.ColorTransition) {constB=2}
  n22 = Saturate(None=Divide(A=Subtract(A=@n19, B=@n21), B=Subtract(B=@n20.ColorTransition)))
  MaterialFunctionCall(Emissive=Multiply(A=ParticleColor().RGB, B=LinearInterpolate(A=@n15, B=@n18.RGB, Alpha=@n22))) {materialFunction=&MF_FX_EmissiveAdjustments}
  n23 = ScalarParameter() {group=01-Opacity, parameterName=SideFadeStart, sortPriority=65}
  n24 = Add(A=@n23, B=ScalarParameter() {defaultValue=0.5, group=01-Opacity, parameterName=SideFadeSmoothness, sortPriority=65})
  n25 = Abs(None=NamedRerouteUsage())
  n26 = ScalarParameter() {defaultValue=0.1, group=01-Opacity, parameterName=SideFadeDistortion, sortPriority=68}
  n27 = NamedRerouteUsage()
  n28 = Round(None=ComponentMask(None=@n27) {g=true})
  n29 = Add(A=Multiply(A=Multiply(A=@n25, B=@n26) {constB=0.1}, B=ConstantBiasScale(None=@n28) {bias=-0.5, scale=2}), B=@n27)
  n30 = ComponentMask(None=@n29) {r=true}
  n31 = Min(A=@n30, B=OneMinus(None=@n30))
  n32 = ScalarParameter() {group=01-Opacity, parameterName=TopFadeStart, sortPriority=65}
  n33 = ComponentMask(None=@n29) {g=true}
  n34 = SmoothStep(
   Min=@n32
   Max=Add(A=@n32, B=ScalarParameter() {defaultValue=0.3, group=01-Opacity, parameterName=TopFadeSmoothness, sortPriority=65})
   Value=@n33
  ) {constMax=0.3}
  n35 = ScalarParameter() {group=01-Opacity, parameterName=BottomFadeStart, sortPriority=65}
  n36 = SmoothStep(
   Min=@n35
   Max=Add(A=@n35, B=ScalarParameter() {defaultValue=0.1, group=01-Opacity, parameterName=BottomFadeSmoothness, sortPriority=65})
   Value=OneMinus(None=@n33)
  ) {constMax=0.1}
  n37 = ScalarParameter() {defaultValue=0.2, group=01-Opacity, parameterName=SwipeDistortion}
  n38 = NamedRerouteUsage()
  n39 = ChannelMaskParameter(None=Add(A=Multiply(A=@n37, B=@n38), B=NamedRerouteUsage())) {group=01-Opacity, maskChannel=Green, parameterName=SwipeDirection, defaultValue=<rgba 0 1 0 0>}
  n40 = DynamicParameter() {defaultValue=<rgba 0.5 0.5 0 1>, paramNames [UV1 U Offset, Swipe, ColorTransition, LinearMask]}
  n41 = ScalarParameter() {defaultValue=0.5, group=01-Opacity, parameterName=SwipeSize, sortPriority=75}
  n42 = Multiply(
   A=Multiply(A=SmoothStep(Min=@n23, Max=@n24, Value=@n31) {constMax=0.5}, B=Min(A=@n34, B=@n36))
   B=Saturate(None=MaterialFunctionCall!(Gradient=@n39.R, Mask Position=@n40.Swipe, Mask Size=@n41) {materialFunction=/Game/Effects/Fort_Effects/MaterialFunctions/MF_GradientMask}.Sinusoidal)
  )
  n43 = DepthFade(FadeDistance=ScalarParameter() {defaultValue=350, group=01-Opacity, parameterName=DepthFadeDist})
  n44 = Constant() {r=1}
  Multiply(A=Multiply(A=ParticleColor().A, B=Saturate(None=@n42)), B=FeatureLevelSwitch(Default=@n43, ES2=@n44, ES3_1=@n44, SM5=@n43, SM6=@n43))
  n45 = TextureCoordinate()
  n46 = LinearInterpolate(
   A=@n45
   B=MaterialFunctionCall(XY=@n45) {materialFunction=&Swizzle}.YX
   Alpha=ScalarParameter() {group=00 - UVSettings, parameterName=SwizzleUVs}
  )
  n47 = Add(A=@n46, B=Constant2Vector() {g=-0.5, r=-0.5})
  NamedRerouteDeclaration(None=LinearInterpolate(
    A=@n46
    B=MaterialFunctionCall(Vector or UVs=@n47) {materialFunction=&VectorToRadialValue}.Radial Coordinates
    Alpha=ScalarParameter() {group=00 - UVSettings, parameterName=UseRadialUVS}
   )) {name=UV0, nodeColor=<rgba 0 1 0.1328 1>}
  n48 = Add(
   A=DynamicParameter() {defaultValue=<rgba 0.5 0.5 0 1>, paramNames [UV1 U Offset, Swipe, ColorTransition, LinearMask]}.UV1 U Offset
   B=VertexColor().A
  )
  n49 = Constant()
  n50 = NamedRerouteUsage()
  n51 = VectorParameter() {group=06- Distortion, parameterName=DistoUV_Settings, channelNames {a=PanY, b=PanX, g=ScaleY, r=ScaleX}, defaultValue=<rgba 1 1 0 0.8>}
  n52 = Panner(Coordinate=@n50, Speed=AppendVector(A=@n51.PanX, B=@n51.PanX)) {speedX=-0.5}
  n53 = AppendVector(A=@n51.ScaleX, B=@n51.ScaleX)
  n54 = TextureSampleParameter2D(UVs=Add(A=AppendVector(A=@n48, B=@n49), B=Multiply(A=@n52, B=@n53))) {
   group=06- Distortion
   parameterName=Distortion Texture3
   samplerType=SAMPLERTYPE_LinearGrayscale
   texture=&T_CellNoise_Tile_M
  }
  n55 = Add(A=ComponentMask(None=@n50) {g=true}) {constB=-0.5}
  NamedRerouteDeclaration(None=Multiply(A=@n54.R, B=Multiply(A=@n55) {constB=-1})) {name=Distortion, nodeColor=<rgba 1 0.3083 0.7353 1>}
  n56 = DynamicParameter() {parameterIndex=1, defaultValue=<rgba 0.5 0.5 0 1>, paramNames [SplineThicknessMult, "", "", ""]}
  n57 = ScalarParameter() {group=07- WPO, parameterName=SplineWidthBase}
  n58 = ScalarParameter() {group=07- WPO, parameterName=SplineWidthTip}
  n59 = TextureCoordinate()
  StaticSwitchParameter(
   True=MaterialFunctionCall(
    WidthBase=Multiply(A=@n56.SplineThicknessMult, B=@n57)
    WidthTip=Multiply(A=@n56.SplineThicknessMult, B=@n58)
    UVs for Projection=@n59
    UVs for Thickness=@n59
    UVs for texturing=@n59
    Expand U or V UV Channel=StaticBool()
   ) {materialFunction=&SplineThicken}.WorldPositionOffset
   False=Constant()
  ) {group=07- WPO, parameterName=UseSplineThicken}
 ]
}
