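# Festive Niagara backdrop for BrickSoup with confetti, glittering sparks, aurora ribbons, glows and a global light, timed to music.
# System values such as Beat, BeatModulo, AudioReactivity and FadeOut drive the emitters, with song BPM exposed as a user parameter.
# On selected beats, confetti bursts from preset points and CPU/GPU spark spirals wrap trees at hardcoded offsets.
# Noise-animated aurora rings pulse with loop and audio, while time of day dims confetti.
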NiagaraSystem /BrickSoup/Assets/FX/NS_BrickSoup_FestiveBackGround
dependencies {
 usedDataInterfaces [
  &NiagaraDataInterfaceColorCurve
  &NiagaraDataInterfaceArrayFloat3
  &NiagaraDataInterfaceCurve
  &NiagaraDataInterfaceArrayFloat
  &NiagaraDataInterfaceSpriteRendererInfo
  &NiagaraDataInterfaceArrayInt32
 ]
 usedDynamicInputs [
  &Divide_Float
  /Niagara/DynamicInputs/Math/Floor_Float
  &ClampFloat
  &Waveform
  &RemapRange
  &Add_Integer
  /Niagara/DynamicInputs/Divide/Divide_Int
 ]
 usedModules [
  &self:SystemSpawnScript.SystemScriptSource.SystemScriptGraph.NiagaraNodeAssignment_6.SetVariables_#1
  &self:SystemSpawnScript.SystemScriptSource.SystemScriptGraph.NiagaraNodeAssignment_7.SetVariables_#2
  &self:SystemSpawnScript.SystemScriptSource.SystemScriptGraph.NiagaraNodeAssignment_1.SetVariables_#3
  &SystemState
  &self:SystemSpawnScript.SystemScriptSource.SystemScriptGraph.NiagaraNodeAssignment_3.SetVariables_#4
  &self:SystemSpawnScript.SystemScriptSource.SystemScriptGraph.NiagaraNodeAssignment_2.SetVariables_#5
  &self:SystemSpawnScript.SystemScriptSource.SystemScriptGraph.NiagaraNodeAssignment_9.SetVariables_#6
  &self:SystemSpawnScript.SystemScriptSource.SystemScriptGraph.NiagaraNodeAssignment_0.SetVariables_#7
  &self:SystemSpawnScript.SystemScriptSource.SystemScriptGraph.NiagaraNodeAssignment_4.SetVariables_#8
  &self:SystemSpawnScript.SystemScriptSource.SystemScriptGraph.NiagaraNodeAssignment_8.SetVariables_#9
  &self:SystemSpawnScript.SystemScriptSource.SystemScriptGraph.NiagaraNodeAssignment_5.SetVariables_#10
  &self:SystemSpawnScript.SystemScriptSource.SystemScriptGraph.NiagaraNodeAssignment_10.SetVariables_#11
  &self:SystemSpawnScript.SystemScriptSource.SystemScriptGraph.NiagaraNodeAssignment_11.SetVariables_#12
 ]
}
emitters {
 ConfettiStreamers {
  emitterUpdateScript [
   &EmitterState()
   &SpawnBurst_Instantaneous(
    Spawn Count=7
    Spawn Probability=&MakeCustomFloatFromBool(
     Bool=&SetBoolByIntComparison(A=$System.BeatModulo, B=0, Comparison Type=::NewEnumerator4)
     True Float=1
     False Float=0
    )
    Use Spawn Probability=true
    UseLoopCountLimit=false
   )
   /SetVariables(Emitter.RandomFloat:NiagaraFloat=&RandomRangeFloat(Minimum=0, Maximum=1)) {isSetParameters=true}
   /SetVariables(Emitter.RandomInt:NiagaraInt32=&ModuloInt(A=$System.Beat, B=5)) {isSetParameters=true}
  ]
  particleSpawnScript [
   &InitializeParticle(
    Color=&ColorFromCurve(
     ColorCurve=DataInterface() {propertyValues {
       AlphaCurve {keys [<at 0 1 RCIM_Cubic>]}
       BlueCurve {keys [<at 0.3165 0 RCIM_Cubic>,<at 0.3584 0 RCIM_Cubic>,<at 0.6275 0>,<at 0.6562 0>]}
       GreenCurve {keys [
         <at 0.3165 0 RCIM_Cubic>
         {arriveTangent=2.3249, interpMode=RCIM_Cubic, leaveTangent=2.3249, time=0.3584, value=0.7631}
         <at 0.6275 0.723>
         <at 0.6562 1>
        ]}
       RedCurve {keys [<at 0.3165 1 RCIM_Cubic>,<at 0.3584 1 RCIM_Cubic>,<at 0.6275 1>,<at 0.6562 0.0364>]}
      }}
     CurveIndex=&RandomRangeFloat(Minimum=0, Maximum=1)
    )
    Lifetime=&RandomRangeFloat(Minimum=1, Maximum=2)
    Mesh Scale Mode=::NewEnumerator3
    Mesh Uniform Scale=&RandomRangeFloat(Minimum=1, Maximum=2)
    Position Offset=&Add_Vector(
     A=$System.TreeOffset
     B=&SelectVectorFromArray(
      Direct Array Index=$Emitter.RandomInt
      Vector Selection Array=DataInterface() {propertyValues {FloatData [<-200 -200 0>, <100 -300 1>, <500 250 -50>, <-400 350 -50>, <450 -450 -30>]}}
     )
    )
    UsePositionOffset=true
    Write Material Random=true
   )
   &ShapeLocation(
    Cylinder Distribution=::NewEnumerator2
    Cylinder Height=0
    Cylinder Radius=5
    MaskSpawnGroup=false
    Shape Primitive=::NewEnumerator1
    U Position=&RandomRangeFloat(Minimum=0, Maximum=1)
    UseOwnerScale=true
    V Position=1
   )
   &AddVelocity(
    Cone Angle=30
    Cone Axis=&RandomRangeVector(Minimum=<-0.5 -0.5 1>, Maximum=<0.5 0.5 1>)
    Module.UseConeFalloff=true
    Module.UseVelDistribution=true
    Override Seed=false
    Velocity Mode=::NewEnumerator2
    Velocity Speed=&RandomRangeFloat(Minimum=444, Maximum=1250)
   )
   &InitialMeshOrientation(
    Mesh Orientation Mode=::NewEnumerator2
    Orientation Coordinate Space=::Local
    Orientation Vector=$Particles.Velocity
    Rotation=<0.125 0 0>
    Use Rotation Vector=true
   )
  ]
  particleUpdateScript [
   &ParticleState()
   &GravityForce(Gravity=<0 0 -111>)
   &Drag(Drag=9, Use Linear Drag=true, Use Rotational Drag=false)
   &SolveForcesAndVelocity(Clamp Velocity=false, Limit Acceleration=false)
   &ScaleColor(
    Scale RGB=&VectorFromFloat(Value=&FNTimeOfDayScalar(Night Scale=0.3, Morning Scale=0.333, Day Scale=0.666, Evening Scale=0.333))
    ScaleA=false
    ScaleRGB=true
   )
  ]
  rendererClasses [&NiagaraMeshRendererProperties]
  renderers [{rendererClass=&NiagaraMeshRendererProperties, rendererIndex=0}]
 }
 NEGlobalLight {
  emitterUpdateScript [
   &EmitterState(
    Inactive Response=::NewEnumerator1
    Life Cycle Mode=::NewEnumerator1
    Loop Behavior=::NewEnumerator1
    Loop Duration Mode=::NewEnumerator1
    UseLoopDelay=false
   )
   &SpawnBurst_Instantaneous(Use Spawn Probability=false, UseLoopCountLimit=false)
  ]
  particleSpawnScript [&InitializeParticle(
    Color Mode=::NewEnumerator0
    Lifetime=5
    Sprite Size Mode=::NewEnumerator3
    Uniform Sprite Size=500
    UsePositionOffset=false
    Write Material Random=true
   )]
  particleUpdateScript [
   &ParticleState(Kill Particles When Lifetime Has Elapsed=false)
   &FN_Light_Attributes(
    Color=&ColorFromCurve(
     ColorCurve=DataInterface() {propertyValues {
       AlphaCurve {keys [<at 0 1 RCIM_Cubic>]}
       BlueCurve {keys [<at 0 1 RCIM_Cubic>,<at 1 1.1>]}
       GreenCurve {keys [<at 0 0.4153 RCIM_Cubic>,<at 1 1.0251>]}
       RedCurve {keys [<at 0 0 RCIM_Cubic>,<at 1 0>]}
      }}
     CurveIndex=&Sine(Normalized Angle=$System.NormalizedLoopAge, Period=1, Scale=1, Bias=0)
    )
    Create Color Attribute=true
    Create Exponent Attribute=true
    Create Position Attribute=true
    Create Radius Attribute=true
    Exponent=1
    Position=<0 0 1500>
    Radius=2500
    Scale Color=4
   )
  ]
  rendererClasses [&NiagaraSpriteRendererProperties, &NiagaraLightRendererProperties]
  renderers [
   {rendererClass=&NiagaraSpriteRendererProperties}
   {rendererClass=&NiagaraLightRendererProperties, rendererIndex=1}
  ]
 }
 NE_AmbientSparks {
  simTarget=GPUComputeSim
  emitterUpdateScript [
   &EmitterState()
   /SetVariables(Emitter.RandomInteger:NiagaraInt32=&RandomRangeInt(Minimum=0, Maximum=2)) {isSetParameters=true}
   &SpawnRate(SpawnRate=&Multiply_Float(A=75, B=$System.FadeOut), Use Spawn Probability=false)
  ]
  particleSpawnScript [
   &InitializeParticle(
    Color=&RandomRangeLinearColor(Minimum=<rgba 1 0.6889 0 1>, Maximum=<rgba 1 0.3404 0 1>)
    Lifetime Max=6
    Lifetime Min=5
    Lifetime Mode=::NewEnumerator1
    Sprite Rotation Mode=::NewEnumerator2
    Sprite Size Mode=::NewEnumerator3
    Uniform Sprite Size=&RandomRangeFloat(Minimum=60, Maximum=80)
    UsePositionOffset=true
    Write Material Random=true
   )
   &ShapeLocation(
    Cylinder Height=800
    Cylinder Radius=2000
    MaskSpawnGroup=false
    Offset=<0 0 350>
    Offset Mode=::NewEnumerator0
    Override Seed=false
    Shape Primitive=::NewEnumerator1
    Surface Only=false
    UseOwnerScale=true
   )
   /SetVariables(Particles.VisibilityTag=&RandomRangeInt(Minimum=0, Maximum=1)) {isSetParameters=true}
   /SetVariables(Particles.RandomBeatInt:NiagaraInt32=&RandomRangeInt(Minimum=0, Maximum=1)) {isSetParameters=true}
  ]
  particleUpdateScript [
   /SetVariables(Particles.GlitterAlpha:NiagaraFloat=&FloatFromCurve(
     FloatCurve=DataInterface() {propertyValues {Curve {keys [
         {interpMode=RCIM_Cubic, tangentMode=RCTM_User, time=0, value=0}
         {interpMode=RCIM_Cubic, tangentMode=RCTM_User, time=0.05, value=1}
         {interpMode=RCIM_Cubic, tangentMode=RCTM_User, time=0.25, value=0}
        ]}}}
     CurveIndex=&ModuloFloat(
      A=&Add_Float(
       A=&Multiply_Float(A=$Emitter.Age, B=2)
       B=&RandomRangeFloat(Minimum=0, Maximum=1)
      )
      B=1
     )
    )) {enabled=false, isSetParameters=true, module=SetVariables_#13}
   /SetVariables(Particles.GlitterAlpha:NiagaraFloat=&FloatFromCurve(
     FloatCurve=DataInterface() {propertyValues {Curve {keys [
         {interpMode=RCIM_Cubic, tangentMode=RCTM_User, time=0, value=0}
         {interpMode=RCIM_Cubic, tangentMode=RCTM_User, time=0.1, value=1}
         {interpMode=RCIM_Cubic, tangentMode=RCTM_User, time=0.5, value=0}
        ]}}}
     CurveIndex=&ModuloFloat(
      A=&Add_Float(
       A=$System.NormalizedLoopAge
       B=&Multiply_Float(A=&MakeFloatFromInt(Int=$Particles.RandomBeatInt), B=0.5)
      )
      B=1
     )
    )) {isSetParameters=true}
   &ParticleState()
   &ScaleSpriteSize(
    Uniform Curve Scale=&Multiply_Float(
     A=&Lerp_Float(A=1, B=1.5, Alpha=$Particles.GlitterAlpha)
     B=&FN_LOD_Distance_Remap(Start Distance=0, End Distance=250, Minimum Output=0, Maximum Output=1, Clamp Output=true)
    )
    Uniform Curve Sprite Scale=DataInterface() {propertyValues {Curve {keys [
        {arriveTangent=19.4641, interpMode=RCIM_Cubic, leaveTangent=19.4641, tangentMode=RCTM_User, time=0, value=0}
        {interpMode=RCIM_Cubic, tangentMode=RCTM_User, time=0.05, value=1}
        {arriveTangent=-3.5627, interpMode=RCIM_Cubic, leaveTangent=-3.5627, tangentMode=RCTM_User, time=1, value=0}
       ]}}}
   )
   &SpriteRotationRate(Rotation Rate=&RandomRangeFloat(Minimum=-20, Maximum=20))
   &ScaleColor(
    Scale Alpha=0.5
    Scale RGB=&Multiply_Vector(
     A=&MakeVectorFromLinearColorRGB(Color=&LerpLinearColors(StartColor=<rgba 1 1 1 1>, EndColor=<rgba 10 6.8887 0 1>, LerpFactor=$Particles.GlitterAlpha))
     B=<1 1 1>
    )
    ScaleA=true
    ScaleRGB=true
   )
   &GravityForce(Gravity=<0 0 -10>)
   /SetVariables(Particles.FLareSize:Vector2f=&Multiply_Vector2D_ByFloat(
     Float=&Lerp_Float(A=0, B=1, Alpha=$Particles.GlitterAlpha)
     Vector2D=&Multiply_Vector2D(A=$Particles.SpriteSize, B=<0.5 0.5>)
    )) {isSetParameters=true, module=SetVariables_#14}
   /SetVariables(Particles.FlareColor=&ColorFromCurve(
     ColorCurve=DataInterface() {propertyValues {
       AlphaCurve {keys [<at 0 0 RCIM_Cubic>,<at 1 0.2 RCIM_Cubic>]}
       BlueCurve {keys [<at 0 0 RCIM_Cubic>,<at 1 0 RCIM_Cubic>]}
       GreenCurve {keys [<at 0 0.383 RCIM_Cubic>,<at 1 0.4596 RCIM_Cubic>]}
       RedCurve {keys [<at 0 1 RCIM_Cubic>,<at 1 1 RCIM_Cubic>]}
      }}
     CurveIndex=$Particles.GlitterAlpha
    )) {isSetParameters=true, module=SetVariables_#15}
   /SetVariables(Particles.SnowFlakeColor:LinearColor=&ColorFromCurve(ColorCurve=DataInterface() {propertyValues {
       AlphaCurve {keys [<at 0 1 RCIM_Cubic>,<at 0.1612 0.2>,<at 1 0 RCIM_Cubic>]}
       BlueCurve {keys [<at 0 0 RCIM_Cubic>,<at 0.1799 1 RCIM_Cubic>]}
       GreenCurve {keys [<at 0 5.9792 RCIM_Cubic>,<at 0.1799 0.9044 RCIM_Cubic>]}
       RedCurve {keys [<at 0 10 RCIM_Cubic>,<at 0.1799 0.4219 RCIM_Cubic>]}
      }})) {isSetParameters=true}
   /SetVariables(Particles.SnowFlakeSize:Vector2f=&Multiply_Vector2D_ByFloat(Float=0.2, Vector2D=$Particles.SpriteSize)) {isSetParameters=true}
   /SetVariables(Particles.SnowFlakeSubUV=1) {isSetParameters=true}
   &Drag(Drag=&RandomRangeFloat(Minimum=1, Maximum=2), Use Linear Drag=true, Use Rotational Drag=true)
   &CurlNoiseForce(Bias Noise=false, Noise Strength=200, Pan Noise=false)
   &SolveForcesAndVelocity(Clamp Velocity=false, Limit Acceleration=false)
   &DynamicMaterialParameters(
    Index 0 Param 1=1
    Index 0 Param 2=0.5
    Module.Index0Param0Write=true
    Module.Index0Param1Write=true
    Module.Index0Param2Write=true
    Module.Index0Param3Write=true
   )
  ]
  rendererClasses [&NiagaraSpriteRendererProperties]
  renderers [
   {rendererClass=&NiagaraSpriteRendererProperties}
   {rendererClass=&NiagaraSpriteRendererProperties, rendererIndex=1}
   {rendererClass=&NiagaraSpriteRendererProperties, rendererIndex=2}
  ]
 }
 NE_AuroraBorealis {
  emitterUpdateScript [
   &EmitterState(
    Inactive Response=::NewEnumerator1
    Life Cycle Mode=::NewEnumerator1
    Loop Behavior=::NewEnumerator1
    Loop Duration Mode=::NewEnumerator1
    UseLoopDelay=true
   )
   &SpawnBurst_Instantaneous(Spawn Count=30, Use Spawn Probability=false, UseLoopCountLimit=false)
  ]
  particleSpawnScript [
   /SetVariables(Particles.NormalizedSpotValue:NiagaraFloat=&FloatFromCurve(
     FloatCurve=DataInterface() {propertyValues {Curve {keys [<at 0 0>,<at 0.5 1>,<at 1 0>]}}}
     CurveIndex=&ReturnNormalizedExecIndex()
    )) {isSetParameters=true}
   &PartitionParticles(
    ID to Partition:NiagaraInt32=&ReturnExecIndex()
    Number Of Partitions=3
    Particles Per Partition=10
    Partition Mode:ENiagara_PartitionMode=::NewEnumerator1
   )
   &InitializeParticle(
    Color=&ColorFromCurve(
     ColorCurve=DataInterface() {propertyValues {
       AlphaCurve {keys [<at 0 1 RCIM_Cubic>]}
       BlueCurve {keys [<at 0 0>,<at 1 0>]}
       GreenCurve {keys [<at 0 0.5965>,<at 1 1.085>]}
       RedCurve {keys [<at 0 2>,<at 1 2>]}
      }}
     CurveIndex=$Particles.NormalizedSpotValue
    )
    Lifetime=&RandomRangeFloat(Minimum=2, Maximum=2)
    Mesh Scale=<1 1 2>
    Mesh Scale Mode=::NewEnumerator4
    Ribbon Width=1000
    Ribbon Width Mode=::NewEnumerator1
    Sprite Size Mode=::NewEnumerator3
    Uniform Sprite Size=50
    UsePositionOffset=true
    Write Material Random=true
   )
   &CameraOffset(Camera Offset Amount=4)
   &self:GetCircularLocation(
    CircleFreq=1
    CircleRadius:NiagaraFloat=&SelectFloatFromArray(
     Direct Array Index=$Particles.PartitionParticles.Partition
     Float Selection Array=DataInterface() {propertyValues {FloatData [2000,1500,1000]}}
    )
    LocationAlpha:NiagaraFloat=$Particles.PartitionParticles.Normalized Index In Partition
    Zoffset:NiagaraFloat=&Add_Float(
     A=&FloatFromCurve(
      FloatCurve=DataInterface() {propertyValues {Curve {keys [
          <at 0 50 RCIM_Cubic>
          {arriveTangent=1416.6666, interpMode=RCIM_Cubic, leaveTangent=1416.6666, time=0.2, value=300}
          {arriveTangent=2125, interpMode=RCIM_Cubic, leaveTangent=2125, time=0.3, value=475}
          {arriveTangent=2000.0001, interpMode=RCIM_Cubic, leaveTangent=2000.0001, time=0.4, value=725}
          {arriveTangent=-125, interpMode=RCIM_Cubic, leaveTangent=-125, time=0.5, value=875}
          {arriveTangent=-2375.0002, interpMode=RCIM_Cubic, leaveTangent=-2375.0002, time=0.6, value=700}
          {arriveTangent=-561.5807, interpMode=RCIM_Cubic, leaveTangent=-561.5807, tangentMode=RCTM_User, time=0.7, value=400}
          <at 1 275 RCIM_Cubic>
         ]}}}
      CurveIndex=$Particles.PartitionParticles.Normalized Index In Partition
     )
     B=&SelectFloatFromArray(
      Direct Array Index=$Particles.PartitionParticles.Partition
      Float Selection Array=DataInterface() {propertyValues {FloatData [500,1000,1500]}}
     )
    )
   ) {module=GetCircularLocation}
   &SubUVAnimation(
    Direct Frame Index:NiagaraFloat=&SelectFloatFromArray(
     Direct Array Index=&ReturnExecIndex()
     Direct Array Mode=::NewEnumerator1
     Float Selection Array=DataInterface() {propertyValues {FloatData [0,0]}}
    )
    "Sprite Renderer "=DataInterface() {propertyValues {SpriteRenderer=&self:NE_AuroraBorealis_2.NiagaraSpriteRendererProperties_2}}
    SubUV Animation Mode:ENiagara_SubUVLookupModeV2=::NewEnumerator5
   )
   &AddVelocity(Velocity=<0 0 1>)
   /SetVariables() {isSetParameters=true}
   &InitialMeshOrientation(Mesh Orientation Mode=::NewEnumerator7, Use Rotation Vector=true)
  ]
  particleUpdateScript [
   &ParticleState(Kill Particles When Lifetime Has Elapsed=false)
   /Niagara/Modules/Update/Utility/VectorNoiseValue(Noise Frequency=1, Pan Noise=true, Pan Noise Field:Vector3f=<0.1 0 0>)
   /SetVariables() {isSetParameters=true}
   /SetVariables(Particles.Position=&Lerp_Vector(
     A=Linked() {linkedVariable {name=Particles.Initial.Position, type=NiagaraPosition}}
     B=&AddVectorToPosition(
      Position=$Particles.Initial.Position
      Vector=&Multiply_VectorByFloat(Vector=$Output.VectorNoiseValue.Noise, Float=150)
     )
     Alpha=&FloatFromCurve(
      FloatCurve=DataInterface() {propertyValues {Curve {keys [
          {interpMode=RCIM_Cubic, tangentMode=RCTM_User, time=0, value=0}
          {interpMode=RCIM_Cubic, tangentMode=RCTM_User, time=0.5, value=1}
          {interpMode=RCIM_Cubic, tangentMode=RCTM_User, time=1, value=0}
         ]}}}
      CurveIndex=&ModuloFloat(
       A=&Multiply_Float(
        A=&Add_Float(A=$Engine.System.Age, B=&RandomRangeFloat(Minimum=0, Maximum=1))
        B=0.1
       )
       B=1
      )
     )
    )) {isSetParameters=true}
   &ScaleColor(
    Scale Alpha=&Multiply_Float(
     A=0.1
     B=&FloatFromCurve(
      FloatCurve=DataInterface() {propertyValues {Curve {keys [
          {interpMode=RCIM_Cubic, tangentMode=RCTM_User, time=0, value=0}
          {arriveTangent=2.0291, interpMode=RCIM_Cubic, leaveTangent=2.0291, tangentMode=RCTM_User, time=1, value=1}
         ]}}}
      CurveIndex=$System.FadeOut
     )
    )
    Scale RGB=&MakeVectorFromLinearColorRGB(Color=&Multiply_LinearColorByFloat(
      Linear Color=&Multiply_LinearColorByFloat(
       Linear Color=<rgba 1 1 1 1>
       Float=&Add_Float(A=1, B=&Multiply_Float(A=$System.AudioReactivity, B=3))
      )
      Float=&Add_Float(A=1, B=$System.Oscillation)
     ))
    ScaleA=true
    ScaleRGB=true
   )
   &SolveForcesAndVelocity(Clamp Velocity=false, Limit Acceleration=false)
   &ScaleRibbonWidth(Ribbon Width Scale=&FloatFromCurve(
     FloatCurve=DataInterface() {propertyValues {Curve {keys [
         {arriveTangent=-2, interpMode=RCIM_Cubic, time=0, value=0.5}
         {arriveTangent=-0.2, interpMode=RCIM_Cubic, leaveTangent=-0.2, time=0.25, value=1}
         {arriveTangent=0.5, interpMode=RCIM_Cubic, leaveTangent=0.5, time=0.5, value=0.4}
         {arriveTangent=0.2, interpMode=RCIM_Cubic, leaveTangent=0.2, time=0.75, value=1.25}
         {interpMode=RCIM_Cubic, tangentMode=RCTM_User, time=1, value=0.5}
        ]}}}
     CurveIndex=&ModuloFloat(A=&ReturnNormalizedExecIndex(Normalized Index Scale=2.5), B=1)
    ))
   &DynamicMaterialParameters(
    Index 0 Param 1=&FloatFromCurve(
     FloatCurve=DataInterface() {propertyValues {Curve {keys [
         {arriveTangent=4.7832, interpMode=RCIM_Cubic, leaveTangent=4.7832, tangentMode=RCTM_User, time=0, value=0}
         {interpMode=RCIM_Cubic, tangentMode=RCTM_User, time=0.25, value=1}
         {interpMode=RCIM_Cubic, tangentMode=RCTM_User, time=1, value=0}
        ]}}}
     CurveIndex=$System.NormalizedLoopAge
    )
    Index 0 Param 2=&Multiply_FloatByInt(Float=2.5, Integer=$System.Beat)
    Module.Index0Param0Write=true
    Module.Index0Param1Write=true
    Module.Index0Param2Write=true
    Module.Index0Param3Write=true
   )
  ]
  rendererClasses [&NiagaraRibbonRendererProperties]
  renderers [{rendererClass=&NiagaraRibbonRendererProperties}]
 }
 NE_BackGroundGlow {
  emitterUpdateScript [
   &EmitterState(
    Inactive Response=::NewEnumerator1
    Life Cycle Mode=::NewEnumerator1
    Loop Behavior=::NewEnumerator1
    Loop Duration Mode=::NewEnumerator1
    UseLoopDelay=false
   )
   &SpawnBurst_Instantaneous(Use Spawn Probability=false, UseLoopCountLimit=false)
  ]
  particleSpawnScript [&InitializeParticle(
    Color=<rgba 2 0.9434 0 0.15>
    Position Offset=<0 0 300>
    Sprite Size Mode=::NewEnumerator3
    Uniform Sprite Size=1500
    UsePositionOffset=true
    Write Material Random=true
   )]
  particleUpdateScript [
   &ParticleState(Kill Particles When Lifetime Has Elapsed=false, Let Infinitely Lived Particles Die When Emitter Deactivates=true)
   &CameraOffset(Camera Offset Amount=-100)
   &ScaleColor(
    Scale Alpha=&Multiply_Float(A=$System.FadeOut, B=&Lerp_Float(A=1, B=1.5, Alpha=$System.AudioReactivity))
    ScaleA=true
    ScaleRGB=true
   )
  ]
  rendererClasses [&NiagaraSpriteRendererProperties]
  renderers [{rendererClass=&NiagaraSpriteRendererProperties}]
 }
 NE_BodyGlow {
  emitterUpdateScript [
   &EmitterState(
    Inactive Response=::NewEnumerator1
    Life Cycle Mode=::NewEnumerator1
    Loop Behavior=::NewEnumerator1
    Loop Duration Mode=::NewEnumerator1
    UseLoopDelay=false
   )
   &SpawnBurst_Instantaneous(Use Spawn Probability=false, UseLoopCountLimit=false)
  ]
  particleSpawnScript [&InitializeParticle(Mesh Scale=<3 3 3>, Mesh Scale Mode=::NewEnumerator4, Position Offset=<0 0 250>, UsePositionOffset=true, Write Material Random=true)]
  particleUpdateScript [
   &ParticleState(Kill Particles When Lifetime Has Elapsed=false, Let Infinitely Lived Particles Die When Emitter Deactivates=true)
   &CameraOffset()
   &ScaleColor(Scale Alpha=$System.FadeOut, ScaleA=true, ScaleRGB=true)
  ]
  rendererClasses [&NiagaraMeshRendererProperties]
  renderers [{rendererClass=&NiagaraMeshRendererProperties, rendererIndex=0}]
 }
 NE_TreesSparks_CPU {
  emitterUpdateScript [
   &EmitterState()
   /SetVariables(Emitter.BaseTreeIndex:NiagaraInt32=&ModuloInt(A=$System.Beat, B=4)) {isSetParameters=true, module=SetVariables_#16}
   &SpawnBurst_Instantaneous(
    Spawn Count=&SelectIntFromArray(
     Direct Array Index=&ModuloInt(A=$System.BeatModulo, B=3)
     Direct Array Mode=::NewEnumerator1
     Int Selection Array=DataInterface() {propertyValues {IntData [40,60]}}
    )
    Spawn Probability=&MakeCustomFloatFromBool(
     Bool=&Boolean_OrOperation(
      Bool1=&SetBoolByIntComparison(A=$System.BeatModulo, B=1, Comparison Type=::NewEnumerator4)
      Bool2=&SetBoolByIntComparison(A=$System.BeatModulo, B=3, Comparison Type=::NewEnumerator4)
     )
     True Float=1
     False Float=0
    )
    Use Spawn Probability=true
    UseLoopCountLimit=false
   )
   /SetVariables(Emitter.RandomFloat:NiagaraFloat=&RandomRangeFloat(Minimum=0, Maximum=1)) {isSetParameters=true, module=SetVariables_#17}
  ]
  particleSpawnScript [
   &PartitionParticles(ID to Partition:NiagaraInt32=&ReturnExecIndex(), Number Of Partitions=3)
   /SetVariables(Particles.TreeIndex:NiagaraInt32=&MergeHelper_IntSwitch(
     0=&SelectIntFromArray(
      Array Sampling Mode=::NewEnumerator1
      Direct Array Index=?
      Int Selection Array=DataInterface() {propertyValues {IntData [0,2,4]}}
      Random Seed=0
     )
     1=&SelectIntFromArray(
      Array Sampling Mode=::NewEnumerator1
      Direct Array Index=?
      Int Selection Array=DataInterface() {propertyValues {IntData [1,3]}}
      Random Seed=0
     )
     2=?
     3=?
     4=?
     5=?
     Index=&ModuloInt(A=$System.Beat, B=4)
     Num=::NewEnumerator0
    )) {isSetParameters=true, module=SetVariables_#18}
   /SetVariables(Particles.TreeScale:NiagaraFloat=&SelectFloatFromArray(Direct Array Index=$Particles.TreeIndex, Float Selection Array=DataInterface() {propertyValues {FloatData [0.375,0.3,0.7,0.35,0.308]}})) {isSetParameters=true, module=SetVariables_#19}
   &InitializeParticle(
    Color=&RandomRangeLinearColor(Minimum=<rgba 1 0.6889 0 1>, Maximum=<rgba 1 0.3404 0 1>)
    Lifetime Min=1.5
    Lifetime Mode=::NewEnumerator1
    Position Offset=&Add_Vector(
     A=$System.TreeOffset
     B=&SelectVectorFromArray(
      Direct Array Index=$Particles.TreeIndex
      Direct Array Mode=::NewEnumerator1
      Vector Selection Array=DataInterface() {propertyValues {FloatData [<-670 -40 0>, <-620 -410 -5>, <150 -880 90>, <760 -120 -20>, <540 585 -55>]}}
     )
    )
    Sprite Rotation Mode=::NewEnumerator2
    Sprite Size Mode=::NewEnumerator3
    Uniform Sprite Size=&Multiply_Float(A=&RandomRangeFloat(Minimum=200, Maximum=300), B=$Particles.TreeScale)
    UsePositionOffset=true
    Write Material Random=true
   )
   /SetVariables(Particles.InitialUniformExecIndex:NiagaraFloat=&ReturnNormalizedExecIndex()) {isSetParameters=true, module=SetVariables_#20}
   /SetVariables(Particles.Age:NiagaraFloat=&Multiply_Float(A=$Particles.InitialUniformExecIndex, B=-0.5)) {isSetParameters=true, module=SetVariables_#21}
   &RotateAroundPoint(
    Radius:NiagaraFloat=&Multiply_Float(
     A=&Multiply_Float(A=450, B=$Particles.TreeScale)
     B=&OneMinusFloat(Float=$Particles.InitialUniformExecIndex)
    )
    Rotation Center:NiagaraPosition=&AddVectorToPosition(
     Position=$Particles.Initial.Position
     Vector=&Multiply_VectorByFloat(
      Vector=&Lerp_Vector(A=<0 0 50>, B=<0 0 1000>, Alpha=$Particles.InitialUniformExecIndex)
      Float=$Particles.TreeScale
     )
    )
    Rotation Phase:NiagaraFloat=&FloatFromCurve(
     FloatCurve=DataInterface() {propertyValues {Curve {keys [<at 0 0>,<at 1 1>]}}}
     CurveIndex=&ModuloFloat(A=&Multiply_Float(A=$Particles.InitialUniformExecIndex, B=3), B=1)
    )
   )
   &AddVelocity(
    Velocity=$Particles.RotateAroundPoint.VectorFromCenter
    Velocity Speed Scale=&RandomRangeFloat(Minimum=25, Maximum=50)
   )
   &AddVelocity(Velocity=<0 0 1>, Velocity Speed Scale=&RandomRangeFloat(Minimum=25, Maximum=50)) {module=AddVelocity001}
  ]
  particleUpdateScript [
   /SetVariables(Particles.GlitterAlpha:NiagaraFloat=&FloatFromCurve(
     FloatCurve=DataInterface() {propertyValues {Curve {keys [
         {interpMode=RCIM_Cubic, tangentMode=RCTM_User, time=0, value=0}
         {interpMode=RCIM_Cubic, tangentMode=RCTM_User, time=0.05, value=1}
         {interpMode=RCIM_Cubic, tangentMode=RCTM_User, time=0.25, value=0}
        ]}}}
     CurveIndex=&ModuloFloat(
      A=&Add_Float(
       A=&Multiply_Float(A=$Emitter.Age, B=0.5)
       B=&RandomRangeFloat(Minimum=0, Maximum=1)
      )
      B=1
     )
    )) {isSetParameters=true, module=SetVariables_#13}
   &ParticleState()
   &ScaleSpriteSize(Uniform Curve Sprite Scale=DataInterface() {propertyValues {Curve {keys [
        {arriveTangent=19.4641, interpMode=RCIM_Cubic, leaveTangent=19.4641, tangentMode=RCTM_User, time=0, value=0}
        {interpMode=RCIM_Cubic, tangentMode=RCTM_User, time=0.05, value=1}
        {arriveTangent=-3.5627, interpMode=RCIM_Cubic, leaveTangent=-3.5627, tangentMode=RCTM_User, time=1, value=0}
       ]}}})
   &SpriteRotationRate(Rotation Rate=&RandomRangeFloat(Minimum=-20, Maximum=20))
   &ScaleColor(
    Scale Alpha=0.5
    Scale RGB=&MakeVectorFromLinearColorRGB(Color=&LerpLinearColors(StartColor=<rgba 1 1 1 1>, EndColor=<rgba 10 6.8887 0 1>, LerpFactor=$Particles.GlitterAlpha))
    ScaleA=true
    ScaleRGB=true
   )
   /SetVariables(Particles.FLareSize:Vector2f=&Multiply_Vector2D_ByFloat(
     Float=&Lerp_Float(A=0, B=1, Alpha=$Particles.GlitterAlpha)
     Vector2D=&Multiply_Vector2D(A=$Particles.SpriteSize, B=<1 1>)
    )) {isSetParameters=true, module=SetVariables_#14}
   /SetVariables(Particles.FlareColor=&ColorFromCurve(
     ColorCurve=DataInterface() {propertyValues {
       AlphaCurve {keys [<at 0 0 RCIM_Cubic>,<at 1 0.2 RCIM_Cubic>]}
       BlueCurve {keys [<at 0 0 RCIM_Cubic>,<at 1 0 RCIM_Cubic>]}
       GreenCurve {keys [<at 0 0.383 RCIM_Cubic>,<at 1 0.4596 RCIM_Cubic>]}
       RedCurve {keys [<at 0 1 RCIM_Cubic>,<at 1 1 RCIM_Cubic>]}
      }}
     CurveIndex=$Particles.GlitterAlpha
    )) {isSetParameters=true, module=SetVariables_#15}
   &CurlNoiseForce(Bias Noise=false, Pan Noise=false)
   &Drag(Use Linear Drag=true, Use Rotational Drag=true)
   &SolveForcesAndVelocity(Clamp Velocity=false, Limit Acceleration=false)
   &DynamicMaterialParameters(
    Index 0 Param 1=1
    Index 0 Param 2=0.5
    Module.Index0Param0Write=true
    Module.Index0Param1Write=true
    Module.Index0Param2Write=true
    Module.Index0Param3Write=true
   )
  ]
  rendererClasses [&NiagaraSpriteRendererProperties]
  renderers [
   {rendererClass=&NiagaraSpriteRendererProperties}
   {rendererClass=&NiagaraSpriteRendererProperties, rendererIndex=1}
  ]
 }
 NE_TreesSparks_GPU001 {
  simTarget=GPUComputeSim
  emitterUpdateScript [
   &EmitterState()
   /SetVariables(Emitter.BaseTreeIndex:NiagaraInt32=&ModuloInt(A=$System.Beat, B=4)) {isSetParameters=true, module=SetVariables_#16}
   &SpawnBurst_Instantaneous(
    Spawn Count=&SelectIntFromArray(
     Direct Array Index=&ModuloInt(A=$System.BeatModulo, B=3)
     Direct Array Mode=::NewEnumerator1
     Int Selection Array=DataInterface() {propertyValues {IntData [450,300]}}
    )
    Spawn Probability=&MakeCustomFloatFromBool(
     Bool=&Boolean_OrOperation(
      Bool1=&SetBoolByIntComparison(A=$System.BeatModulo, B=1, Comparison Type=::NewEnumerator4)
      Bool2=&SetBoolByIntComparison(A=$System.BeatModulo, B=3, Comparison Type=::NewEnumerator4)
     )
     True Float=1
     False Float=0
    )
    Use Spawn Probability=true
    UseLoopCountLimit=false
   )
   /SetVariables(Emitter.RandomFloat:NiagaraFloat=&RandomRangeFloat(Minimum=0, Maximum=1)) {isSetParameters=true, module=SetVariables_#17}
  ]
  particleSpawnScript [
   &PartitionParticles(ID to Partition:NiagaraInt32=&ReturnExecIndex(), Number Of Partitions=3)
   /SetVariables(Particles.TreeIndex:NiagaraInt32=&MergeHelper_IntSwitch(
     0=&SelectIntFromArray(
      Array Sampling Mode=::NewEnumerator1
      Direct Array Index=?
      Int Selection Array=DataInterface() {propertyValues {IntData [0,2,4]}}
      Random Seed=0
     )
     1=&SelectIntFromArray(
      Array Sampling Mode=::NewEnumerator1
      Direct Array Index=?
      Int Selection Array=DataInterface() {propertyValues {IntData [1,3]}}
      Random Seed=0
     )
     2=?
     3=?
     4=?
     5=?
     Index=&ModuloInt(A=$System.Beat, B=4)
     Num=::NewEnumerator0
    )) {isSetParameters=true, module=SetVariables_#18}
   /SetVariables(Particles.TreeScale:NiagaraFloat=&SelectFloatFromArray(Direct Array Index=$Particles.TreeIndex, Float Selection Array=DataInterface() {propertyValues {FloatData [0.375,0.3,0.7,0.35,0.308]}})) {isSetParameters=true, module=SetVariables_#19}
   &InitializeParticle(
    Color=&RandomRangeLinearColor(Minimum=<rgba 1 0.6889 0 1>, Maximum=<rgba 1 0.3404 0 1>)
    Lifetime Min=1.5
    Lifetime Mode=::NewEnumerator1
    Position Offset=&Add_Vector(
     A=$System.TreeOffset
     B=&SelectVectorFromArray(
      Direct Array Index=$Particles.TreeIndex
      Direct Array Mode=::NewEnumerator1
      Vector Selection Array=DataInterface() {propertyValues {FloatData [<-670 -40 30>, <-620 -410 25>, <150 -880 90>, <760 -120 10>, <540 585 -25>]}}
     )
    )
    Sprite Rotation Mode=::NewEnumerator2
    Sprite Size Mode=::NewEnumerator3
    Uniform Sprite Size=&Multiply_Float(A=&RandomRangeFloat(Minimum=200, Maximum=300), B=$Particles.TreeScale)
    UsePositionOffset=true
    Write Material Random=true
   )
   /SetVariables(Particles.InitialUniformExecIndex:NiagaraFloat=&ReturnNormalizedExecIndex()) {isSetParameters=true, module=SetVariables_#20}
   /SetVariables(Particles.Age:NiagaraFloat=&Multiply_Float(A=$Particles.InitialUniformExecIndex, B=-0.5)) {isSetParameters=true, module=SetVariables_#21}
   &RotateAroundPoint(
    Radius:NiagaraFloat=&Multiply_Float(
     A=&Multiply_Float(A=450, B=$Particles.TreeScale)
     B=&OneMinusFloat(Float=$Particles.InitialUniformExecIndex)
    )
    Rotation Center:NiagaraPosition=&AddVectorToPosition(
     Position=$Particles.Initial.Position
     Vector=&Multiply_VectorByFloat(
      Vector=&Lerp_Vector(A=<0 0 50>, B=<0 0 1000>, Alpha=$Particles.InitialUniformExecIndex)
      Float=$Particles.TreeScale
     )
    )
    Rotation Phase:NiagaraFloat=&FloatFromCurve(
     FloatCurve=DataInterface() {propertyValues {Curve {keys [<at 0 0>,<at 1 1>]}}}
     CurveIndex=&ModuloFloat(A=&Multiply_Float(A=$Particles.InitialUniformExecIndex, B=3), B=1)
    )
   )
   &AddVelocity(
    Velocity=$Particles.RotateAroundPoint.VectorFromCenter
    Velocity Speed Scale=&RandomRangeFloat(Minimum=25, Maximum=50)
   )
   &AddVelocity(Velocity=<0 0 1>, Velocity Speed Scale=&RandomRangeFloat(Minimum=25, Maximum=50)) {module=AddVelocity001}
  ]
  particleUpdateScript [
   /SetVariables(Particles.GlitterAlpha:NiagaraFloat=&FloatFromCurve(
     FloatCurve=DataInterface() {propertyValues {Curve {keys [
         {interpMode=RCIM_Cubic, tangentMode=RCTM_User, time=0, value=0}
         {interpMode=RCIM_Cubic, tangentMode=RCTM_User, time=0.05, value=1}
         {interpMode=RCIM_Cubic, tangentMode=RCTM_User, time=0.25, value=0}
        ]}}}
     CurveIndex=&ModuloFloat(
      A=&Add_Float(
       A=&Multiply_Float(A=$Emitter.Age, B=0.5)
       B=&RandomRangeFloat(Minimum=0, Maximum=1)
      )
      B=1
     )
    )) {isSetParameters=true, module=SetVariables_#13}
   &ParticleState()
   &ScaleSpriteSize(Uniform Curve Sprite Scale=DataInterface() {propertyValues {Curve {keys [
        {arriveTangent=19.4641, interpMode=RCIM_Cubic, leaveTangent=19.4641, tangentMode=RCTM_User, time=0, value=0}
        {interpMode=RCIM_Cubic, tangentMode=RCTM_User, time=0.05, value=1}
        {arriveTangent=-3.5627, interpMode=RCIM_Cubic, leaveTangent=-3.5627, tangentMode=RCTM_User, time=1, value=0}
       ]}}})
   &SpriteRotationRate(Rotation Rate=&RandomRangeFloat(Minimum=-20, Maximum=20))
   &ScaleColor(
    Scale Alpha=0.5
    Scale RGB=&MakeVectorFromLinearColorRGB(Color=&LerpLinearColors(StartColor=<rgba 1 1 1 1>, EndColor=<rgba 10 6.8887 0 1>, LerpFactor=$Particles.GlitterAlpha))
    ScaleA=true
    ScaleRGB=true
   )
   /SetVariables(Particles.FLareSize:Vector2f=&Multiply_Vector2D_ByFloat(
     Float=&Lerp_Float(A=0, B=1, Alpha=$Particles.GlitterAlpha)
     Vector2D=&Multiply_Vector2D(A=$Particles.SpriteSize, B=<1 1>)
    )) {isSetParameters=true, module=SetVariables_#14}
   /SetVariables(Particles.FlareColor=&ColorFromCurve(
     ColorCurve=DataInterface() {propertyValues {
       AlphaCurve {keys [<at 0 0 RCIM_Cubic>,<at 1 0.2 RCIM_Cubic>]}
       BlueCurve {keys [<at 0 0 RCIM_Cubic>,<at 1 0 RCIM_Cubic>]}
       GreenCurve {keys [<at 0 0.383 RCIM_Cubic>,<at 1 0.4596 RCIM_Cubic>]}
       RedCurve {keys [<at 0 1 RCIM_Cubic>,<at 1 1 RCIM_Cubic>]}
      }}
     CurveIndex=$Particles.GlitterAlpha
    )) {isSetParameters=true, module=SetVariables_#15}
   &CurlNoiseForce(Bias Noise=false, Pan Noise=false)
   &Drag(Use Linear Drag=true, Use Rotational Drag=true)
   &SolveForcesAndVelocity(Clamp Velocity=false, Limit Acceleration=false)
   &DynamicMaterialParameters(
    Index 0 Param 1=1
    Index 0 Param 2=0.5
    Module.Index0Param0Write=true
    Module.Index0Param1Write=true
    Module.Index0Param2Write=true
    Module.Index0Param3Write=true
   )
  ]
  rendererClasses [&NiagaraSpriteRendererProperties]
  renderers [
   {rendererClass=&NiagaraSpriteRendererProperties}
   {rendererClass=&NiagaraSpriteRendererProperties, rendererIndex=1}
  ]
 }
}
summary {
 systemName=NS_BrickSoup_FestiveBackGround
 emitters [
  {emitterName=ConfettiStreamers, rendererClasses [&NiagaraMeshRendererProperties]}
  {emitterName=NE_AmbientSparks, simTarget=GPUComputeSim, rendererClasses [&NiagaraSpriteRendererProperties]}
  {emitterName=NE_AuroraBorealis, rendererClasses [&NiagaraRibbonRendererProperties]}
  {emitterName=NE_TreesSparks_GPU001, simTarget=GPUComputeSim, rendererClasses [&NiagaraSpriteRendererProperties]}
  {emitterName=NE_BodyGlow, rendererClasses [&NiagaraMeshRendererProperties]}
  {emitterName=NE_BackGroundGlow, rendererClasses [&NiagaraSpriteRendererProperties]}
  {emitterName=NE_TreesSparks_CPU, rendererClasses [&NiagaraSpriteRendererProperties]}
  {emitterName=NEGlobalLight, rendererClasses [&NiagaraSpriteRendererProperties, &NiagaraLightRendererProperties]}
 ]
 userVariables [User.SongBPM {type=NiagaraFloat, defaultValue=75, description=""}]
}
