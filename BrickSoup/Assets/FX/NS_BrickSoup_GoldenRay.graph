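# Niagara system for the BrickSoup golden ray, a light beam with five states shifting from blue to gold.
# It layers a darkening sprite, ray mesh, ribbon meshes, sparkle mesh and glittering curl-noise sparks, spawned only in the top two states.
# Arrays indexed by NextShapeIndex or User.RayState pick per-state values, with colors and scales easing at StateSwitchSpeed.
# System timeline, explosion scale and camera-distance remaps modulate growth, size and visibility.
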NiagaraSystem /BrickSoup/Assets/FX/NS_BrickSoup_GoldenRay
dependencies {
 usedDataInterfaces [
  &NiagaraDataInterfaceArrayColor
  &NiagaraDataInterfaceArrayFloat
  &NiagaraDataInterfaceCurve
  &NiagaraDataInterfaceVector2DCurve
  &NiagaraDataInterfaceColorCurve
  &NiagaraDataInterfaceArrayInt32
 ]
 usedDynamicInputs [
  &SelectIntFromArray
  &MakeIntFromFloat
  &MakeFloatFromInt
 ]
 usedModules [
  &self:SystemSpawnScript.SystemScriptSource.SystemScriptGraph.NiagaraNodeAssignment_1.SetVariables_#1
  &self:SystemSpawnScript.SystemScriptSource.SystemScriptGraph.NiagaraNodeAssignment_2.SetVariables_#2
  &self:SystemSpawnScript.SystemScriptSource.SystemScriptGraph.NiagaraNodeAssignment_3.SetVariables_#3
  &self:SystemSpawnScript.SystemScriptSource.SystemScriptGraph.NiagaraNodeAssignment_4.SetVariables_#4
  &self:SystemSpawnScript.SystemScriptSource.SystemScriptGraph.NiagaraNodeAssignment_5.SetVariables_#5
  &SystemState
  &self:SystemSpawnScript.SystemScriptSource.SystemScriptGraph.NiagaraNodeAssignment_9.SetVariables_#6
  &self:SystemSpawnScript.SystemScriptSource.SystemScriptGraph.NiagaraNodeAssignment_0.SetVariables_#7
  &Timeline
  &self:SystemSpawnScript.SystemScriptSource.SystemScriptGraph.NiagaraNodeAssignment_6.SetVariables_#8
  &self:SystemSpawnScript.SystemScriptSource.SystemScriptGraph.NiagaraNodeAssignment_7.SetVariables_#9
  &self:SystemSpawnScript.SystemScriptSource.SystemScriptGraph.NiagaraNodeAssignment_8.SetVariables_#10
  &self:SwitchCurrentShape
 ]
}
emitters {
 NE_Darken {
  emitterUpdateScript [
   &EmitterState(
    Inactive Response=::NewEnumerator1
    Life Cycle Mode=::NewEnumerator1
    Loop Behavior=::NewEnumerator1
    Loop Duration Mode=::NewEnumerator1
    UseLoopDelay=false
   )
   &SpawnBurst_Instantaneous(Use Spawn Probability=false, UseLoopCountLimit=false)
  ]
  particleSpawnScript [
   &InitializeParticle(
    Color=<rgba 0.025 0.0082 0 1>
    Lifetime=5
    Sprite Size=<3000 15000>
    Sprite Size Mode=::NewEnumerator4
    UsePositionOffset=true
    Write Material Random=true
   )
   /SetVariables(Particles.SpriteAlignment=<0 0 1>) {isSetParameters=true}
  ]
  particleUpdateScript [
   &ParticleState(Kill Particles When Lifetime Has Elapsed=false)
   &DynamicMaterialParameters(
    Index 0 Param 1=1.25
    Index 0 Param 2=1
    Index 0 Param 3=1
    Index 0 Param 4=1
    Module.Index0Param0Write=true
    Module.Index0Param1Write=true
    Module.Index0Param2Write=true
    Module.Index0Param3Write=true
   )
   /SetVariables(Particles.Color=&MultiplyLinearColor(
     Linear Color=&LerpLinearColors(
      StartColor=$Particles.Color
      EndColor=&SelectLinearColorFromArray(
       Array Sampling Mode=::NewEnumerator1
       Direct Array Index=$System.NextShapeIndex
       Color Selection Array=DataInterface() {propertyValues {ColorData [
          <rgba 0.0094 0.0343 0.05 0.3>
          <rgba 0.0228 0.0759 0.1094 0.5>
          <rgba 0.0228 0.0759 0.1094 0.5>
          <rgba 0.0228 0.0759 0.1094 0.5>
          <rgba 0.0228 0.0759 0.1094 0.5>
         ]}}
       Random Seed=?
      )
      LerpFactor=&Multiply_Float(A=$Engine.DeltaTime, B=$User.StateSwitchSpeed)
     )
     Scale Factor=&MakeFloat4FromVectorAndFloat(
      "Vector (RGB)"=<1 1 1>
      "Float (Alpha)"=&FloatFromCurve(FloatCurve=DataInterface() {propertyValues {Curve {keys [<at 0 0>,<at 1 1>]}}}, CurveIndex=$System.Timeline.NormalizedTime)
     )
    )) {isSetParameters=true}
   &ScaleSpriteSize(
    Non-Uniform Curve Index:NiagaraFloat=$System.Timeline.NormalizedTime
    Non-Uniform Curve Sprite Scale=DataInterface() {propertyValues {
      XCurve {keys [{tangentMode=RCTM_User, time=0, value=1}, {tangentMode=RCTM_User, time=1, value=1}]}
      YCurve {keys [{tangentMode=RCTM_User, time=0, value=0}, {tangentMode=RCTM_User, time=1, value=1}]}
     }}
    Scale Sprite Size Mode=::NewEnumerator3
   )
  ]
  rendererClasses [&NiagaraSpriteRendererProperties]
  renderers [{rendererClass=&NiagaraSpriteRendererProperties}]
 }
 NE_GoldRay {
  emitterSpawnScript [/SetVariables() {isSetParameters=true}]
  emitterUpdateScript [
   &EmitterState(
    Inactive Response=::NewEnumerator1
    Life Cycle Mode=::NewEnumerator1
    Loop Behavior=::NewEnumerator1
    Loop Duration Mode=::NewEnumerator1
    UseLoopDelay=false
   )
   /SetVariables(Emitter.XUVOffset:NiagaraFloat=&Add_Float(
     A=$Emitter.XUVOffset
     B=&Multiply_Float(
      A=$Engine.DeltaTime
      B=&SelectFloatFromArray(Direct Array Index=$System.NextShapeIndex, Float Selection Array=DataInterface() {propertyValues {FloatData [0,0,0,-0.5,-0.75]}})
     )
    )) {isSetParameters=true}
   &SpawnBurst_Instantaneous(Use Spawn Probability=false, UseLoopCountLimit=false)
  ]
  particleSpawnScript [
   &InitializeParticle(
    Lifetime=5
    Mesh Scale=<1.25 1 1>
    Mesh Scale Mode=::NewEnumerator4
    Sprite Size=<800 1500>
    Sprite Size Mode=::NewEnumerator4
    UsePositionOffset=true
    Write Material Random=true
   )
   &InitialMeshOrientation(Mesh Orientation Mode=::NewEnumerator7, Rotation=<0 -0.25 0>, Use Rotation Vector=true)
   /SetVariables() {isSetParameters=true}
  ]
  particleUpdateScript [
   &ParticleState(Kill Particles When Lifetime Has Elapsed=false)
   /SetVariables(Particles.StateScale:NiagaraFloat=&Lerp_Float(
     A=$Particles.StateScale
     B=&SelectFloatFromArray(Direct Array Index=$System.NextShapeIndex, Float Selection Array=DataInterface() {propertyValues {FloatData [70,40,25,30,30]}})
     Alpha=&Multiply_Float(A=$Engine.DeltaTime, B=$User.StateSwitchSpeed)
    )) {isSetParameters=true}
   &DynamicMaterialParameters(
    Index 0 Param 1=&Multiply_Float(A=$Particles.StateScale, B=$System.Timeline.NormalizedTime)
    Index 0 Param 3=$Emitter.XUVOffset
    Index 0 Param 4=1
    Module.Index0Param0Write=true
    Module.Index0Param1Write=true
    Module.Index0Param2Write=true
    Module.Index0Param3Write=true
   )
   /SetVariables(Particles.Color=&MultiplyLinearColor(
     Linear Color=&LerpLinearColors(
      StartColor=$Particles.Color
      EndColor=&SelectLinearColorFromArray(
       Array Sampling Mode=::NewEnumerator1
       Direct Array Index=$System.NextShapeIndex
       Color Selection Array=DataInterface() {propertyValues {ColorData [<rgba 0.1875 0.6859 1 0.5>, <rgba 0.2083 0.6939 1 1>, <rgba 0.1375 0.7892 1.2 1>, <rgba 0 1.0256 2 1>, <rgba 1.5 0.5923 0 1>]}}
       Random Seed=?
      )
      LerpFactor=&Multiply_Float(A=$Engine.DeltaTime, B=$User.StateSwitchSpeed)
     )
     Scale Factor=&MakeFloat4FromVectorAndFloat(
      "Vector (RGB)"=<1 1 1>
      "Float (Alpha)"=&FloatFromCurve(FloatCurve=DataInterface() {propertyValues {Curve {keys [<at 0 0>,<at 1 1>]}}}, CurveIndex=$System.Timeline.NormalizedTime)
     )
    )) {isSetParameters=true}
  ]
  rendererClasses [&NiagaraMeshRendererProperties]
  renderers [{rendererClass=&NiagaraMeshRendererProperties, rendererIndex=0}]
 }
 NE_RainbowRibbons {
  emitterUpdateScript [
   &EmitterState()
   &SpawnRate(
    SpawnRate=&SelectFloatFromArray(Direct Array Index=$System.NextShapeIndex, Float Selection Array=DataInterface() {propertyValues {FloatData [1,1,1,1,3]}})
    Use Spawn Probability=false
   )
  ]
  particleSpawnScript [
   &InitializeParticle(
    Lifetime=&MergeHelper_FloatSwitch(
     0=0
     1=0
     2=0
     3=&RandomRangeFloat(Minimum=4, Maximum=5)
     4=&RandomRangeFloat(Minimum=2, Maximum=2.5)
     Index=$System.NextShapeIndex
     Num=::NewEnumerator3
    )
    Mass Mode=::NewEnumerator3
    Mesh Scale Mode=::NewEnumerator3
    Mesh Uniform Scale=&Multiply_Float(A=&RandomRangeFloat(Minimum=3, Maximum=4), B=$System.ExposionScale)
    Position Offset=$System.BaseOffset
    UsePositionOffset=true
    Write Material Random=true
   )
   &CameraOffset(Camera Offset Amount=4)
   &InitialMeshOrientation(
    Mesh Axis To Orient=<0 1 0>
    Mesh Orientation Mode=::NewEnumerator2
    Orientation Vector=<0 0 1>
    Rotation=&RandomRangeVector(Minimum=<0 0 0>, Maximum=<0 1 0>)
    Use Rotation Vector=true
   )
   /SetVariables() {isSetParameters=true}
   /SetVariables() {isSetParameters=true}
   /SetVariables(Particles.Color=&SelectLinearColorFromArray(
     Array Sampling Mode=::NewEnumerator1
     Direct Array Index=$System.NextShapeIndex
     Color Selection Array=DataInterface() {propertyValues {ColorData [<rgba 0.1875 0.6859 1 0>, <rgba 0.2083 0.6939 1 0>, <rgba 1 1 1 0>, <rgba 0.1458 0.7175 1 0.15>, <rgba 1 0.7475 0.5104 0.15>]}}
     Random Seed=?
    )) {isSetParameters=true}
  ]
  particleUpdateScript [
   &ParticleState()
   &ScaleMeshSize(Scale Factor=&Multiply_VectorByFloat(Vector=<1 1 1>, Float=1))
   /SetVariables(Particles.StateScale:NiagaraFloat=&Lerp_Float(
     A=$Particles.StateScale
     B=&SelectFloatFromArray(Direct Array Index=$System.NextShapeIndex, Float Selection Array=DataInterface() {propertyValues {FloatData [0,0,0,5,10]}})
     Alpha=&Multiply_Float(A=$Engine.DeltaTime, B=$User.StateSwitchSpeed)
    )) {isSetParameters=true}
   /SetVariables(Particles.StateMiscFloat:NiagaraFloat=&Lerp_Float(
     A=$Particles.StateMiscFloat
     B=&SelectFloatFromArray(Direct Array Index=$System.NextShapeIndex, Float Selection Array=DataInterface() {propertyValues {FloatData [0,0,0,0,1]}})
     Alpha=&Multiply_Float(A=$Engine.DeltaTime, B=$User.StateSwitchSpeed)
    )) {isSetParameters=true}
   &DynamicMaterialParameters(
    Index 0 Param 2=&FloatFromCurve(FloatCurve=DataInterface() {propertyValues {Curve {keys [<at 0 1.1>,<at 1 0>]}}})
    Index 0 Param 3=&Lerp_Float(
     A=1
     B=&FloatFromCurve(FloatCurve=DataInterface() {propertyValues {Curve {keys [
          {interpMode=RCIM_Cubic, tangentMode=RCTM_User, time=0, value=1}
          {arriveTangent=-1.9555, interpMode=RCIM_Cubic, leaveTangent=-1.9555, tangentMode=RCTM_User, time=1, value=0}
         ]}}})
     Alpha=$Particles.StateMiscFloat
    )
    Index 1 Param 1=&Multiply_Float(A=$Particles.StateScale, B=$System.ExposionScale)
    Module.Index0Param0Write=true
    Module.Index0Param1Write=true
    Module.Index0Param2Write=true
    Module.Index0Param3Write=true
    Module.Index1Param0Write=true
    Module.Index1Param1Write=true
    Module.Index1Param2Write=true
    Module.Index1Param3Write=true
    Write Parameter Index 1=true
   )
   &ScaleColor(
    Scale Alpha=&FN_LOD_Distance_Remap(Start Distance=4000, End Distance=6000, Minimum Output=0, Maximum Output=1, Clamp Output=true)
    ScaleA=true
    ScaleRGB=true
   )
   /SetVariables(Particles.GlowColor=&MultiplyLinearColor(Linear Color=$Particles.Color, Scale Factor=&MakeVector4_FromLinearColor(LinearColor=<rgba 1 1 1 0.4>))) {isSetParameters=true}
  ]
  rendererClasses [&NiagaraMeshRendererProperties]
  renderers [
   {rendererClass=&NiagaraMeshRendererProperties, rendererIndex=0}
   {rendererClass=&NiagaraMeshRendererProperties, rendererIndex=1}
  ]
 }
 NE_Sparkles {
  emitterUpdateScript [
   &EmitterState(
    Inactive Response=::NewEnumerator1
    Life Cycle Mode=::NewEnumerator1
    Loop Behavior=::NewEnumerator1
    Loop Duration Mode=::NewEnumerator1
    UseLoopDelay=false
   )
   &SpawnBurst_Instantaneous(Use Spawn Probability=false, UseLoopCountLimit=false)
  ]
  particleSpawnScript [
   &InitializeParticle(
    Color=<rgba 1 0.5554 0 0.05>
    Lifetime=5
    Mesh Scale=<5 5 10>
    Mesh Scale Mode=::NewEnumerator4
    UsePositionOffset=false
    Write Material Random=true
   )
   &InitialMeshOrientation(
    Mesh Orientation Mode=::NewEnumerator7
    Rotation=&RandomRangeVector(Minimum=<0 0 0>, Maximum=<0 0 1>)
    Use Rotation Vector=true
   )
   /SetVariables() {isSetParameters=true}
  ]
  particleUpdateScript [
   &ParticleState(Kill Particles When Lifetime Has Elapsed=false)
   &ScaleMeshSize(Scale Factor=&Multiply_Vector(A=<1 1 1>, B=&MakeVector(X=$System.ExposionScale, Y=$System.ExposionScale, Z=1)))
   /SetVariables(Particles.Color=&LerpLinearColors(
     StartColor=$Particles.Color
     EndColor=&SelectLinearColorFromArray(
      Array Sampling Mode=::NewEnumerator1
      Direct Array Index=$System.NextShapeIndex
      Color Selection Array=DataInterface() {propertyValues {ColorData [<rgba 0.1875 0.6859 1 0.1>, <rgba 0.2083 0.6939 1 0.2>, <rgba 0.2083 0.6939 1 0.5>, <rgba 0.2083 0.6939 1 0.7>, <rgba 1 0.5751 0 0.2>]}}
      Random Seed=?
     )
     LerpFactor=&Multiply_Float(A=$Engine.DeltaTime, B=$User.StateSwitchSpeed)
    )) {isSetParameters=true}
   /SetVariables(Particles.StateScale:NiagaraFloat=&Lerp_Float(
     A=$Particles.StateScale
     B=&SelectFloatFromArray(Direct Array Index=$System.NextShapeIndex, Float Selection Array=DataInterface() {propertyValues {FloatData [0.5,0.6,0.6,0.6,0.6]}})
     Alpha=&Multiply_Float(A=$Engine.DeltaTime, B=$User.StateSwitchSpeed)
    )) {isSetParameters=true}
   &DynamicMaterialParameters(
    Index 0 Param 1=&Multiply_Float(
     A=&Multiply_Float(
      A=&Multiply_Float(A=&Multiply_Float(A=3, B=$Particles.StateScale), B=$System.ExposionScale)
      B=$System.Timeline.NormalizedTime
     )
     B=&FN_LOD_Distance_Remap(Start Distance=5000, End Distance=8000, Minimum Output=0, Maximum Output=1, Clamp Output=true)
    )
    Index 0 Param 3=1.4
    Index 0 Param 4=1
    Module.Index0Param0Write=true
    Module.Index0Param1Write=true
    Module.Index0Param2Write=true
    Module.Index0Param3Write=true
   )
  ]
  rendererClasses [&NiagaraMeshRendererProperties]
  renderers [{rendererClass=&NiagaraMeshRendererProperties, rendererIndex=0}]
 }
 NE_Sparks {
  emitterUpdateScript [
   &EmitterState()
   &SpawnRate(
    SpawnRate=&SelectFloatFromArray(Direct Array Index=$User.RayState, Float Selection Array=DataInterface() {propertyValues {FloatData [0,0,0,2.5,3.5]}})
    Use Spawn Probability=false
   )
  ]
  particleSpawnScript [
   /SetVariables(Particles.SizeAlpha:NiagaraFloat=&UniformAOrBFloat(A=0, Evaluation Type=::NewEnumerator0, Recalculate Random Each Loop=true, Override Seed=false)) {isSetParameters=true}
   &InitializeParticle(
    Color=&RandomRangeLinearColor(Minimum=<rgba 1 0.6889 0 1>, Maximum=<rgba 1 0.3404 0 1>)
    Lifetime=&Multiply_Float(
     A=&RandomRangeFloat(Minimum=8, Maximum=10)
     B=&SelectFloatFromArray(Direct Array Index=$User.RayState, Float Selection Array=DataInterface() {propertyValues {FloatData [0,0,0,1,1]}})
    )
    Position Offset=&Multiply_VectorByFloat(
     Vector=&RandomRangeVector(Minimum=<-300 -300 500>, Maximum=<300 300 5500>)
     Float=$System.ExposionScale
    )
    Sprite Rotation Mode=::NewEnumerator2
    Sprite Size Mode=::NewEnumerator3
    Uniform Sprite Size=&RandomRangeFloat(Minimum=100, Maximum=120)
    UsePositionOffset=true
    Write Material Random=true
   )
  ]
  particleUpdateScript [
   /SetVariables(Particles.GlitterAlpha:NiagaraFloat=&FloatFromCurve(
     FloatCurve=DataInterface() {propertyValues {Curve {keys [
         {interpMode=RCIM_Cubic, tangentMode=RCTM_User, time=0, value=0}
         {interpMode=RCIM_Cubic, tangentMode=RCTM_User, time=0.05, value=1}
         {interpMode=RCIM_Cubic, tangentMode=RCTM_User, time=0.25, value=0}
        ]}}}
     CurveIndex=&ModuloFloat(
      A=&Add_Float(
       A=&Multiply_Float(A=$Emitter.Age, B=0.25)
       B=&RandomRangeFloat(Minimum=0, Maximum=1)
      )
      B=1
     )
    )) {isSetParameters=true}
   &ParticleState()
   &ScaleSpriteSize(
    Uniform Curve Scale=&RemapRange(Input Value To Remap=$System.ExposionScale, Input Min=1, Input Max=4, Output Min=1, Output Max=7)
    Uniform Curve Sprite Scale=DataInterface() {propertyValues {Curve {keys [
        {arriveTangent=19.4641, interpMode=RCIM_Cubic, leaveTangent=19.4641, tangentMode=RCTM_User, time=0, value=0}
        {interpMode=RCIM_Cubic, tangentMode=RCTM_User, time=0.05, value=1}
        {arriveTangent=-3.5627, interpMode=RCIM_Cubic, leaveTangent=-3.5627, tangentMode=RCTM_User, time=1, value=0}
       ]}}}
   )
   &SpriteRotationRate(Rotation Rate=&RandomRangeFloat(Minimum=-20, Maximum=20))
   /SetVariables(Particles.Color=&MultiplyLinearColor(
     Linear Color=&LerpLinearColors(
      StartColor=$Particles.Color
      EndColor=&SelectLinearColorFromArray(
       Array Sampling Mode=::NewEnumerator1
       Direct Array Index=$System.NextShapeIndex
       Color Selection Array=DataInterface() {propertyValues {ColorData [<rgba 0.1875 0.6859 1 0.1>, <rgba 0.2083 0.6939 1 0.2>, <rgba 0.2083 0.6939 1 0.5>, <rgba 10.4165 34.6953 50 1>, <rgba 50 28.7571 0 1>]}}
       Random Seed=?
      )
      LerpFactor=&Multiply_Float(A=$Engine.DeltaTime, B=$User.StateSwitchSpeed)
     )
     Scale Factor=&MakeFloat4FromVectorAndFloat(
      "Vector (RGB)"=&MakeVectorFromLinearColorRGB(Color=&LerpLinearColors(StartColor=<rgba 1 1 1 1>, EndColor=<rgba 1 1 1 1>, LerpFactor=$Particles.GlitterAlpha))
      "Float (Alpha)"=&Multiply_Float(
       A=0.5
       B=&FN_LOD_Distance_Remap(Start Distance=4000, End Distance=6000, Minimum Output=0, Maximum Output=1, Clamp Output=true)
      )
     )
    )) {isSetParameters=true}
   &ScaleColor(
    Color Value To Scale=$Particles.Color
    Scale Alpha=&Multiply_Float(
     A=0.5
     B=&FN_LOD_Distance_Remap(Start Distance=4000, End Distance=6000, Minimum Output=0, Maximum Output=1, Clamp Output=true)
    )
    Scale RGB=&MakeVectorFromLinearColorRGB(Color=&LerpLinearColors(StartColor=<rgba 1 1 1 1>, EndColor=<rgba 10 10 10 1>, LerpFactor=$Particles.GlitterAlpha))
    ScaleA=true
    ScaleRGB=true
   ) {enabled=false}
   &GravityForce(Gravity=<0 0 -80>)
   /SetVariables(Particles.FLareSize:Vector2f=&Multiply_Vector2D_ByFloat(
     Float=&Lerp_Float(A=0, B=1, Alpha=$Particles.GlitterAlpha)
     Vector2D=&Multiply_Vector2D(A=$Particles.SpriteSize, B=<1 1>)
    )) {isSetParameters=true}
   /SetVariables(Particles.FlareColor=&MultiplyLinearColor(
     Linear Color=$Particles.Color
     Scale Factor=&MakeVector4_FromLinearColor(LinearColor=&ColorFromCurve(
       ColorCurve=DataInterface() {propertyValues {
         AlphaCurve {keys [<at 0 0 RCIM_Cubic>,<at 1 0.2 RCIM_Cubic>]}
         BlueCurve {keys [<at 0 1 RCIM_Cubic>,<at 1 1 RCIM_Cubic>]}
         GreenCurve {keys [<at 0 1 RCIM_Cubic>,<at 1 1 RCIM_Cubic>]}
         RedCurve {keys [<at 0 1 RCIM_Cubic>,<at 1 1 RCIM_Cubic>]}
        }}
       CurveIndex=$Particles.GlitterAlpha
      ))
    )) {isSetParameters=true}
   &Drag(Drag=&RandomRangeFloat(Minimum=1, Maximum=2), Use Linear Drag=true, Use Rotational Drag=true)
   &CurlNoiseForce(Bias Noise=false, Noise Strength=500, Pan Noise=false)
   &SolveForcesAndVelocity(Clamp Velocity=false, Limit Acceleration=false)
   &DynamicMaterialParameters(
    Index 0 Param 1=1
    Index 0 Param 2=0.5
    Module.Index0Param0Write=true
    Module.Index0Param1Write=true
    Module.Index0Param2Write=true
    Module.Index0Param3Write=true
   )
  ]
  rendererClasses [&NiagaraSpriteRendererProperties]
  renderers [
   {rendererClass=&NiagaraSpriteRendererProperties}
   {rendererClass=&NiagaraSpriteRendererProperties, rendererIndex=1}
  ]
 }
}
summary {
 systemName=NS_BrickSoup_GoldenRay
 emitters [
  {emitterName=NE_Sparkles, rendererClasses [&NiagaraMeshRendererProperties]}
  {emitterName=NE_GoldRay, rendererClasses [&NiagaraMeshRendererProperties]}
  {emitterName=NE_Darken, rendererClasses [&NiagaraSpriteRendererProperties]}
  {emitterName=NE_RainbowRibbons, rendererClasses [&NiagaraMeshRendererProperties]}
  {emitterName=NE_Sparks, rendererClasses [&NiagaraSpriteRendererProperties]}
 ]
 userVariables [
  User.BaseShapeIndex {type=NiagaraInt32}
  User.IsVisible {type=NiagaraBool, defaultValue=true}
  User.RayState {type=NiagaraInt32, defaultValue=4}
  User.StateSwitchSpeed {type=NiagaraFloat, defaultValue=0.5, description=""}
 ]
}
