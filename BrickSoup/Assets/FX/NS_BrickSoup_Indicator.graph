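# NS_BrickSoup_Indicator is a Niagara indicator effect for BrickSoup: a large light-blue corona sprite aligned to the Z axis.
# Its single emitter, NE_Corona, burst-spawns particles that persist past their lifetime and feed dynamic parameters to the sprite material.
# A system Timeline module's normalized time, rather than particle age, drives the alpha ramp from transparent to opaque and the sprite size curve.
# It exposes a User.IsVisible bool, defaulting to true.
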NiagaraSystem /BrickSoup/Assets/FX/NS_BrickSoup_Indicator
dependencies {
 usedDataInterfaces [&NiagaraDataInterfaceCurve, &NiagaraDataInterfaceVector2DCurve]
 usedDynamicInputs []
 usedModules [&SystemState, &Timeline]
}
emitters {NE_Corona {
  emitterUpdateScript [
   &EmitterState(
    Inactive Response=::NewEnumerator1
    Life Cycle Mode=::NewEnumerator1
    Loop Behavior=::NewEnumerator1
    Loop Duration Mode=::NewEnumerator1
    UseLoopDelay=false
   )
   &SpawnBurst_Instantaneous(Use Spawn Probability=false, UseLoopCountLimit=false)
  ]
  particleSpawnScript [
   &InitializeParticle(
    Color=<rgba 0.3125 0.746 1 1>
    Lifetime=5
    Sprite Size Mode=::NewEnumerator3
    Uniform Sprite Size=1000
    UsePositionOffset=true
    Write Material Random=true
   )
   /SetVariables(Particles.SpriteAlignment=<0 0 1>) {isSetParameters=true}
  ]
  particleUpdateScript [
   &ParticleState(Kill Particles When Lifetime Has Elapsed=false)
   &DynamicMaterialParameters(
    Index 0 Param 1=1.25
    Index 0 Param 2=0.5
    Index 0 Param 3=1
    Index 0 Param 4=1
    Module.Index0Param0Write=true
    Module.Index0Param1Write=true
    Module.Index0Param2Write=true
    Module.Index0Param3Write=true
   )
   &ScaleColor(
    Scale Alpha=&FloatFromCurve(FloatCurve=DataInterface() {propertyValues {Curve {keys [<at 0 0>,<at 1 1>]}}}, CurveIndex=$System.Timeline.NormalizedTime)
    ScaleA=true
    ScaleRGB=true
   )
   &ScaleSpriteSize(Uniform Curve Index=$System.Timeline.NormalizedTime)
  ]
  rendererClasses [&NiagaraSpriteRendererProperties]
  renderers [{rendererClass=&NiagaraSpriteRendererProperties}]
 }}
summary {
 systemName=NS_BrickSoup_Indicator
 emitters [{emitterName=NE_Corona, rendererClasses [&NiagaraSpriteRendererProperties]}]
 userVariables [User.IsVisible {type=NiagaraBool, defaultValue=true}]
}
