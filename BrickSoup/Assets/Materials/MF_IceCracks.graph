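# BrickSoup material function that layers cracked ice with an animated, glowing core onto incoming material attributes.
# Local triplanar UVs scaled to object bounds drive position-rotated cracks, blended normals and view-offset parallax bubbles.
# Time-of-day colors, fresnel, sun blob and backlight masks shade the ice, while Is_Fragment limits glow by an upward-facing mask.
# Ice color feeds subsurface on the default path, but folds into emissive on mobile and low quality.
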MaterialFunction /BrickSoup/Assets/Materials/MF_IceCracks
graph {nodes [
  n1 = NamedRerouteUsage()
  n2 = Multiply(
   A=@n1
   B=LinearInterpolate(
    A=VectorParameter() {group=BrickSoup, parameterName=CoreGlow_Emissive1, defaultValue=<rgba 1 0.1406 0.0625 1>}.RGB
    B=VectorParameter() {group=BrickSoup, parameterName=CoreGlow_Emissive2, defaultValue=<rgba 1 0.5164 0.1242 1>}.RGB
    Alpha=@n1
   )
  )
  n3 = VectorParameter() {group=BrickSoup, parameterName=Ice_Base Color2_morning, defaultValue=<rgba 0.0125 0.1227 0.3 1>}
  n4 = VectorParameter() {group=BrickSoup, parameterName=Ice_Base Color2_day, defaultValue=<rgba 0.0125 0.1227 0.3 1>}
  n5 = VectorParameter() {group=BrickSoup, parameterName=Ice_Base Color2_Evening, defaultValue=<rgba 0.0125 0.1227 0.3 1>}
  n6 = VectorParameter() {group=BrickSoup, parameterName=Ice_Base Color2_Night, defaultValue=<rgba 0.0125 0.1227 0.3 1>}
  n7 = VectorParameter() {group=BrickSoup, parameterName=Ice_Base Color3_Morning, defaultValue=<rgba 0.1175 0.1026 0.1615 1>}
  n8 = VectorParameter() {group=BrickSoup, parameterName=Ice_Base Color3_Day, defaultValue=<rgba 0.1175 0.1026 0.1615 1>}
  n9 = VectorParameter() {group=BrickSoup, parameterName=Ice_Base Color3_Evening, defaultValue=<rgba 0.1175 0.1026 0.1615 1>}
  n10 = VectorParameter() {group=BrickSoup, parameterName=Ice_Base Color3_Night, defaultValue=<rgba 0.1175 0.1026 0.1615 1>}
  n11 = CameraVectorWS()
  n12 = Normalize(VectorInput=Constant3Vector() {constant=<rgba -0.4 -0.7 0.2 0>})
  n13 = Saturate(None=DotProduct(A=PixelNormalWS(), B=MaterialProxyReplace(Realtime=@n11, MaterialProxy=@n12)))
  n14 = Multiply(A=@n13, B=@n13)
  n15 = Multiply(A=@n14, B=@n14)
  n16 = Multiply(
   A=NamedRerouteDeclaration(None=Multiply(A=@n15, B=@n15)) {name=Ice Fresnel, nodeColor=<rgba 1 0.5391 0 1>}
   B=ScalarParameter() {defaultValue=-0.2, desc=-.12, group=BrickSoup, parameterName=Fresnel Dark Core}
  )
  n17 = NamedRerouteUsage()
  n18 = Multiply(A=Add(A=ComponentMask(None=@n17) {g=true}, B=Constant() {r=0.4}) {constB=-0.5}) {constB=0.5}
  n19 = DotProduct(A=CameraVectorWS(), B=SkyAtmosphereLightDirection())
  n20 = OneMinus(None=@n19)
  n21 = SmoothStep(Value=Add(A=SquareRoot(None=@n20), B=ScalarParameter() {defaultValue=-0.99, group=BrickSoup, parameterName=Sun Blob_Size}) {constB=-0.1}) {constMax=0.44, constMin=0.39}
  n22 = Multiply(A=@n21, B=@n21)
  n23 = NamedRerouteUsage()
  n24 = Multiply(A=Multiply(A=@n22, B=@n22), B=@n23)
  n25 = ScalarParameter() {defaultValue=3, group=BrickSoup, parameterName=Bubbles_Inside_Sun}
  n26 = Add(
   A=NamedRerouteDeclaration(None=Multiply(A=@n24, B=@n25) {constB=10}) {name=SunCracks, nodeColor=<rgba 0.8516 0 1 1>}
   B=Multiply(A=@n23, B=ScalarParameter() {defaultValue=0.4, group=BrickSoup, parameterName=Bubbles_Inside_Opacity}) {constB=0.05}
  )
  n27 = Constant() {r=0.65}
  n28 = Saturate(None=Multiply(A=Add(A=@n19, B=@n27) {constB=-0.1}, B=Constant() {r=-3}) {constB=-0.5})
  n29 = Add(A=@n26, B=Multiply(A=@n26, B=NamedRerouteDeclaration(None=@n28) {name=SSS Mask, nodeColor=<rgba 0 0.1562 1 1>})) {constB=0}
  n30 = Saturate(None=@n29)
  n31 = Saturate(None=DotProduct(
    A=SkyAtmosphereLightDirection()
    B=MaterialFunctionCall() {materialFunction=&CameraDirectionVector}.Result
   ))
  n32 = Multiply(
   A=@n29
   B=Multiply(
    A=NamedRerouteDeclaration(None=@n31) {name=LightDirection, nodeColor=<rgba 0.9453 0 1 1>}
    B=ScalarParameter() {defaultValue=1, group=BrickSoup, parameterName=Against Sun_Mult}
   )
  )
  n33 = NamedRerouteUsage()
  n34 = ScalarParameter() {defaultValue=2, group=BrickSoup, parameterName=IceBase_Edges_Noise}
  n35 = SmoothStep(
   Min=ScalarParameter() {desc=0.8, group=BrickSoup, parameterName=IceBase_Edges_Min}
   Max=ScalarParameter() {defaultValue=4.5, desc=0.2, group=BrickSoup, parameterName=IceBase_EdgesMax}
   Value=NamedRerouteUsage()
  ) {constMax=0.6}
  n36 = Subtract(A=@n35, B=Multiply(A=Multiply(A=@n33, B=@n34), B=NamedRerouteUsage()))
  n37 = NamedRerouteUsage()
  n38 = Saturate(None=Add(A=Add(A=@n35, B=@n36), B=SmoothStep(Value=@n37) {constMin=0.1}))
  n39 = Subtract(
   A=Max(A=Add(A=@n30, B=@n32), B=OneMinus(None=@n38))
   B=Multiply(A=Saturate(None=@n18), B=ScalarParameter() {defaultValue=0.55, group=BrickSoup, parameterName=Vertical Gradient})
  )
  n40 = NamedRerouteUsage()
  n41 = ScalarParameter() {defaultValue=0.3, group=BrickSoup, parameterName=Fragments_Base}
  n42 = LinearInterpolate(
   A=LinearInterpolate(
    A=MaterialFunctionCall(Color_Morning=@n3.RGB, Color_Day=@n4.RGB, Color_Evening=@n5.RGB, Color_Night=@n6.RGB) {materialFunction=&Color_Time_of_Day}.Result
    B=MaterialFunctionCall(Color_Morning=@n7.RGB, Color_Day=@n8.RGB, Color_Evening=@n9.RGB, Color_Night=@n10.RGB) {materialFunction=&Color_Time_of_Day}.Result
    Alpha=NamedRerouteUsage()
   )
   B=VectorParameter() {group=BrickSoup, parameterName=Ice_Base Color1, defaultValue=<rgba 0.151 0.4397 1 1>}.RGB
   Alpha=Add(A=Add(A=@n39, B=@n16), B=Multiply(A=@n40, B=@n41)) {constB=0}
  ) {constB=0}
  n43 = FunctionInput() {inputType=FunctionInput_MaterialAttributes}
  n44 = MaterialFunctionCall(
   BaseNormal=NamedRerouteDeclaration(None=GetMaterialAttributes(None=@n43) {attributeGetTypes [#1, #2]}.Normal) {name=Normals, nodeColor=<rgba 0.7109 0 1 1>}
   AdditionalNormal=NamedRerouteUsage()
  ) {materialFunction=&BlendAngleCorrectedNormals}
  n45 = SetMaterialAttributes!(MaterialAttributes, SubsurfaceColor, EmissiveColor=Add(A=@n42, B=@n2), Opacity, BaseColor, Normal=@n44.Result) {attributeSetTypes [
    #3
    #4
    #5
    #6
    #1
   ]}
  n46 = SetMaterialAttributes(
   SubsurfaceColor=@n42
   EmissiveColor=@n2
   Opacity=ScalarParameter() {defaultValue=-0.4, group=BrickSoup, parameterName=Ice_OpacityBias}
   Normal=@n44.Result
  ) {attributeSetTypes [
    #3
    #4
    #5
    #6
    #1
   ]}
  FunctionOutput(None=ShadingPathSwitch(Default=QualitySwitch(Default=@n46, Low=@n45), Mobile=@n45)) {outputName=MA, sortPriority=-1}
  n47 = NamedRerouteDeclaration(None=Transform(None=VertexNormalWS()) {transformSourceType=TRANSFORMSOURCE_World, transformType=TRANSFORM_Local}) {name=LS Vtx Normals, nodeColor=<rgba 0 1 0.6719 1>}
  n48 = Multiply(
   A=Add(A=Abs(None=@n47), B=Constant() {r=-0.5}) {constB=-0.5}
   B=ScalarParameter() {defaultValue=3, group=BrickSoup, parameterName=Triplanar Fade}
  ) {constB=5}
  n49 = MaterialFunctionCall(Color=Saturate(None=@n48)) {materialFunction=&SplitComponents}
  NamedRerouteDeclaration(None=@n49.RGB) {name=Triplanar Mask, nodeColor=<rgba 0.9688 0 1 1>}
  n50 = MaterialFunctionCall(Color=LocalPosition().XYZ) {materialFunction=&SplitComponents}
  n51 = OneMinus(None=AppendVector(A=@n50.R, B=@n50.G))
  n52 = Add(A=@n47, B=Constant() {r=-0.5}) {constB=-0.5}
  n53 = MaterialFunctionCall(Float3=NamedRerouteDeclaration(None=Sign(None=@n52)) {name=Triplanar Sign, nodeColor=<rgba 0.4766 1 0 1>}) {materialFunction=&BreakOutFloat3Components}
  n54 = Multiply(A=@n49.R, B=AppendVector(A=Multiply(A=@n53.R, B=@n50.G) {constB=-1}, B=@n50.B))
  n55 = Multiply(A=@n53.G, B=@n50.R) {constB=-1}
  n56 = Multiply(A=@n49.G, B=AppendVector(A=OneMinus(None=@n55), B=@n50.B))
  n57 = MaterialFunctionCall(Color=NamedRerouteUsage()) {materialFunction=&SplitComponents}
  n58 = MaterialFunctionCall(Color=ObjectLocalBounds().Extents) {materialFunction=&SplitComponents}
  n59 = Max(A=Multiply(A=@n57.R, B=AppendVector(A=@n58.G, B=@n58.G)), B=Multiply(A=@n57.G, B=AppendVector(A=@n58.R, B=@n58.R)))
  n60 = Multiply(A=@n57.B, B=AppendVector(A=@n58.R, B=@n58.R))
  n61 = Divide(
   A=Add(A=Multiply(A=@n49.B, B=@n51), B=Add(A=@n54, B=@n56))
   B=NamedRerouteDeclaration(None=Max(A=@n59, B=@n60)) {name=UV Scale to bounds, nodeColor=<rgba 1 0 0.1406 1>}
  )
  n62 = MaterialFunctionCall(Color=NamedRerouteUsage()) {materialFunction=&SplitComponents}
  n63 = AppendVector(A=Constant() {r=0.5}, B=NamedRerouteUsage())
  n64 = Multiply(A=@n62.R, B=@n63)
  n65 = Multiply(A=@n62.G, B=@n63)
  n66 = Constant2Vector() {g=0.5, r=0.5}
  n67 = NamedRerouteDeclaration(None=Max(A=Max(A=@n64, B=@n65), B=Multiply(A=@n62.B, B=@n66))) {name=UV Origin, nodeColor=<rgba 0 0.4609 1 1>}
  NamedRerouteDeclaration(None=OneMinus(None=Add(A=@n61, B=@n67))) {name=UV_Triplanar, nodeColor=<rgba 0.8984 0 1 1>}
  n68 = Constant() {r=1}
  n69 = Constant() {r=1.5}
  n70 = MaterialFunctionCall(Color=OneMinus(None=SmoothStep(Min=@n68, Max=@n69, Value=@n48) {constMin=0.8})) {materialFunction=&SplitComponents}
  n71 = Multiply(A=Multiply(A=@n70.R, B=@n70.R), B=@n70.B)
  NamedRerouteDeclaration(None=@n71) {name=Mask_ProjetionBroken_Inv, nodeColor=<rgba 0 0.1562 1 1>}
  NamedRerouteDeclaration(None=TextureObject() {texture=/BrickSoup/Assets/Textures/T_BrickSoup_IceCracks_Bubbles}) {name=Masks_2, nodeColor=<rgba 0 1 0.9766 1>}
  n72 = ObjectLocalBounds()
  n73 = NamedRerouteUsage()
  n74 = AppendVector(A=Constant2Vector(), B=Multiply(A=ComponentMask(None=@n72.Extents) {b=true}, B=Subtract(B=@n73) {constA=0.5}) {constB=0.3})
  NamedRerouteDeclaration(None=MaterialFunctionCall!(Position=TransformPosition(None=@n74) {transformType=TRANSFORMPOSSOURCE_World}, UVRadialScale) {materialFunction=/Game/Weapons/WeaponSkins/MaterialFunction/ObjectStabilizedScreenSpaceUVs}.StabilizedUVs) {name=Billboard UV, nodeColor=<rgba 0.8516 0 1 1>}
  n75 = ScalarParameter() {defaultValue=1.45, group=BrickSoup, parameterName=Dark Core_Lenght}
  n76 = Multiply(
   A=LocalPosition().XYZ
   B=AppendVector(A=AppendVector(A=@n75, B=@n75), B=ScalarParameter() {defaultValue=0.65, group=BrickSoup, parameterName=Dark Core_Height})
  )
  n77 = Constant()
  n78 = ObjectLocalBounds()
  n79 = ComponentMask(None=@n78.Extents) {b=true}
  n80 = NamedRerouteDeclaration(None=Constant() {r=0.27}) {name=Cracks Origin_1, nodeColor=<rgba 0 1 0.6719 1>}
  n81 = MaterialFunctionCall(
   R=@n77
   G=@n77
   B=Multiply(A=NamedRerouteDeclaration(None=@n79) {name=ObjectBounds_Z, nodeColor=<rgba 0.0547 0 1 1>}, B=Subtract(B=@n80) {constA=0.5}) {constB=0.3}
  ) {materialFunction=&AppendMany}
  n82 = Multiply(A=ComponentMask(None=@n78.Extents) {r=true}, B=Constant() {r=1})
  n83 = Constant() {r=0.56}
  NamedRerouteDeclaration(None=Saturate(None=SphereMask(A=@n76, B=@n81.RGB, Radius=@n82, Hardness=@n83) {attenuationRadius=512, hardnessPercent=30})) {name=M_DarkIceCore, nodeColor=<rgba 0 1 0.9063 1>}
  NamedRerouteDeclaration(None=ScalarParameter() {defaultValue=0.18, group=BrickSoup, parameterName=Radial Mask_Z Offset}) {name=Cracks Origin, nodeColor=<rgba 0 1 0.6719 1>}
  n84 = Add(A=NamedRerouteUsage(), B=ComponentMask(None=ObjectPositionWS()) {r=true})
  n85 = MaterialFunctionCall(
   UVs=NamedRerouteUsage()
   "Rotation Angle (0-1)"=NamedRerouteDeclaration(None=Frac(None=@n84)) {name=CrackRotation, nodeColor=<rgba 0 1 0.4375 1>}
  ) {materialFunction=&CustomRotator}
  n86 = ScalarParameter() {defaultValue=1, group=BrickSoup, parameterName=Crack_Tile}
  n87 = TextureSample(
   UVs=NamedRerouteDeclaration(None=Multiply(A=@n85.Rotated Values, B=@n86)) {name=Crack UV, nodeColor=<rgba 0.7812 1 0 1>}
   Tex=NamedRerouteUsage()
  )
  n88 = ScalarParameter() {defaultValue=3, group=BrickSoup, parameterName=Cracks_Blob Hardness}
  n89 = Distance(A=NamedRerouteUsage(), B=Constant() {r=0.5})
  n90 = Subtract(
   A=Add(A=NamedRerouteDeclaration(None=@n89) {name=TriplanarBlob, nodeColor=<rgba 1 0.6797 0 1>})
   B=ScalarParameter() {defaultValue=0.8, group=BrickSoup, parameterName=Cracks_Blob Radius}
  ) {constB=0}
  n91 = LinearInterpolate(A=OneMinus(None=@n88), B=@n88, Alpha=@n90)
  n92 = Multiply(
   A=Subtract(A=NamedRerouteDeclaration(None=@n87.G) {name=T_Cracks, nodeColor=<rgba 1 0.7734 0 1>}, B=Saturate(None=@n91))
   B=NamedRerouteUsage()
  )
  NamedRerouteDeclaration(None=@n92) {name=M_Cracks, nodeColor=<rgba 1 0 0.3984 1>}
  n93 = MaterialFunctionCall(Color=NamedRerouteUsage()) {materialFunction=&SplitComponents}
  n94 = Transform(None=LocalPosition().XYZ) {transformSourceType=TRANSFORMSOURCE_Local}
  n95 = Multiply(A=NamedRerouteUsage(), B=ScalarParameter() {defaultValue=-800, group=BrickSoup, parameterName=Bubbles_BumpOffset_1})
  n96 = ScalarParameter() {defaultValue=-100, group=BrickSoup, parameterName=Bubbles_BumpOffset_2}
  n97 = Multiply(
   A=CameraVectorWS()
   B=Clamp(
    None=Add(A=@n95, B=@n96)
    Min=ScalarParameter() {defaultValue=-1000, group=BrickSoup, parameterName=Bubbles_BumpOffset_Min}
    Max=ScalarParameter() {defaultValue=1000, group=BrickSoup, parameterName=Bubbles_BumpOffset_Max}
   )
  )
  n98 = MaterialFunctionCall(Color=Transform(None=Add(A=@n94, B=@n97)) {transformSourceType=TRANSFORMSOURCE_World, transformType=TRANSFORM_Local}) {materialFunction=&SplitComponents}
  n99 = AppendVector(A=@n98.R, B=@n98.G)
  n100 = MaterialFunctionCall(Float3=NamedRerouteUsage()) {materialFunction=&BreakOutFloat3Components}
  n101 = AppendVector(A=Multiply(A=@n100.R, B=@n98.G) {constB=-1}, B=@n98.B)
  n102 = AppendVector(A=OneMinus(None=Multiply(A=@n100.G, B=@n98.R) {constB=-1}), B=@n98.B)
  n103 = Add(A=Multiply(A=@n93.B, B=OneMinus(None=@n99)), B=Add(A=Multiply(A=@n93.R, B=@n101), B=Multiply(A=@n93.G, B=@n102)))
  n104 = NamedRerouteDeclaration(None=ScalarParameter() {defaultValue=0.0027, group=BrickSoup, parameterName=Bubbles_Inside_Tiling}) {name=Tiling Inside, nodeColor=<rgba 0 1 0.5781 1>}
  NamedRerouteDeclaration(None=TextureSample(UVs=Multiply(A=@n103, B=@n104), Tex=NamedRerouteUsage()) {samplerType=SAMPLERTYPE_LinearColor, texture=/BrickSoup/Assets/Textures/T_BrickSoup_IceCracks_Masks}.R) {name=T_InsideBubbles, nodeColor=<rgba 1 0 0.1406 1>}
  NamedRerouteDeclaration(None=@n87.B) {name=T_Surface_RandomDepht, nodeColor=<rgba 0 0.2734 1 1>}
  NamedRerouteDeclaration(None=@n90) {name=M_CrackRadiusGradient, nodeColor=<rgba 0.5938 0 1 1>}
  n105 = NamedRerouteUsage()
  n106 = Multiply(A=Multiply(A=Clamp(None=@n105) {minDefault=-1}, B=NamedRerouteUsage()), B=Constant3Vector() {constant=<rgba 1 2 3 0>})
  n107 = MaterialFunctionCall(Color=Divide(A=Add(A=@n106)) {constB=6}) {materialFunction=&SplitComponents}
  NamedRerouteDeclaration(None=Add(A=Add(A=@n107.R, B=@n107.R), B=@n107.B)) {name=Random_Faces, nodeColor=<rgba 0.9219 1 0 1>}
  n108 = Divide(A=@n103, B=NamedRerouteUsage())
  n109 = NamedRerouteUsage()
  n110 = NamedRerouteDeclaration(None=OneMinus(None=Add(A=@n108, B=@n109))) {name=UV_RandomFragments, nodeColor=<rgba 0 1 0.2031 1>}
  n111 = NamedRerouteUsage()
  NamedRerouteDeclaration(None=Multiply(
    A=TextureSample(UVs=@n110, Tex=@n111) {samplerType=SAMPLERTYPE_LinearColor, texture=&T_BrickSoup_IceCracks_Masks}.B
    B=NamedRerouteUsage()
   )) {name=T_Surface_RandomDepht_2, nodeColor=<rgba 0 0.2734 1 1>}
  n112 = TextureSample(UVs=NamedRerouteUsage()) {samplerType=SAMPLERTYPE_Normal, texture=/BrickSoup/Assets/Textures/T_BrickSoup_IceCracks_N}
  n113 = ComponentMask(None=@n112.RGB) {g=true, r=true}
  n114 = NamedRerouteUsage()
  n115 = DotProduct(A=@n113, B=AppendVector(A=Cosine(None=@n114), B=Sine(None=@n114)))
  n116 = Add(A=@n114) {constB=0.25}
  n117 = DotProduct(A=@n113, B=AppendVector(A=Cosine(None=@n116), B=Sine(None=@n116)))
  n118 = AppendVector(
   A=Multiply(A=AppendVector(A=@n115, B=@n117), B=ScalarParameter() {defaultValue=1, group=BrickSoup, parameterName=Normal_Crack_Mult})
   B=@n112.B
  )
  n119 = MaterialFunctionCall(
   In=Add(A=NamedRerouteUsage(), B=ScalarParameter() {group=BrickSoup, parameterName=Normal_Mask_Bias})
   Contrast=ScalarParameter() {group=BrickSoup, parameterName=Normal_Mask_Contrast}
  ) {materialFunction=&CheapContrast}
  n120 = Multiply(A=@n119.Result, B=NamedRerouteUsage())
  n121 = NamedRerouteDeclaration(None=StaticBoolParameter() {group=BrickSoup, parameterName=Is_Fragment, sortPriority=0}) {name=IsFragment_1, nodeColor=<rgba 1 0 0.1875 1>}
  n122 = Max(A=OneMinus(None=StaticSwitch(True=@n120, False=@n119.Result, Value=@n121)), B=NamedRerouteUsage())
  NamedRerouteDeclaration(None=Normalize(VectorInput=MaterialFunctionCall(Normal=@n118, Flatness=@n122) {materialFunction=&FlattenNormal}.Result)) {name=Crack Normal, nodeColor=<rgba 0.7578 0 1 1>}
  n123 = ScalarParameter() {defaultValue=-10, group=BrickSoup, parameterName=Core Glow_Point Min}
  n124 = ScalarParameter() {defaultValue=0.5, group=BrickSoup, parameterName=Core Glow_Point Max}
  n125 = NamedRerouteUsage()
  n126 = NamedRerouteUsage()
  n127 = ScalarParameter() {defaultValue=0.6, group=BrickSoup, parameterName=Core Glow_Pos Lerp}
  n128 = Constant()
  n129 = ScalarParameter() {group=BrickSoup, parameterName=Core Glow_Blob Z Offset}
  n130 = Distance(A=Add(A=LinearInterpolate(A=@n125, B=@n126, Alpha=@n127) {constAlpha=0.8}, B=AppendVector(A=@n128, B=@n129)), B=Constant() {r=0.5})
  n131 = OneMinus(None=@n130)
  n132 = Subtract(A=@n131)
  n133 = ScalarParameter() {defaultValue=0.5, group=BrickSoup, parameterName=Core Glow_Blob Size}
  n134 = Add(
   A=Saturate(None=LinearInterpolate(A=@n123, B=@n124, Alpha=@n131))
   B=Multiply(A=Add(A=@n132, B=@n133), B=ScalarParameter() {defaultValue=0.8, group=BrickSoup, parameterName=Core Glow_Blob Contrast})
  )
  n135 = Time()
  n136 = ScalarParameter() {defaultValue=5, group=BrickSoup, parameterName=Core Glow_Anim Speed}
  n137 = ScalarParameter() {defaultValue=-5, group=BrickSoup, parameterName=Core Glow_Anim Freq}
  n138 = Sine(None=Add(A=Divide(A=@n135, B=@n136) {constB=10}, B=Multiply(A=@n130, B=@n137) {constB=-5}))
  n139 = ScalarParameter() {defaultValue=0.02, desc=0.15, group=BrickSoup, parameterName=Core Glow_Anim Mult}
  n140 = NamedRerouteUsage()
  n141 = Add(
   A=Add(A=NamedRerouteDeclaration(None=@n134) {name=InnerGlowBlob, nodeColor=<rgba 0 1 0.5313 1>}, B=Multiply(A=@n138, B=@n139))
   B=Multiply(A=Add(A=@n140) {constB=-0.5}, B=ScalarParameter() {defaultValue=0.6, group=BrickSoup, parameterName=Core Glow_Bubbles}) {constB=0.05}
  )
  n142 = Multiply(
   A=Add(A=NamedRerouteUsage(), B=ScalarParameter() {defaultValue=-0.28, group=BrickSoup, parameterName=Core Glow_Fragments radius bias})
   B=ScalarParameter() {defaultValue=5, group=BrickSoup, parameterName=Core Glow_Fragments Radius Mult}
  )
  n143 = Subtract(
   A=NamedRerouteUsage()
   B=LinearInterpolate(
    A=ScalarParameter() {defaultValue=0.15, group=BrickSoup, parameterName=Core Glow_Fragments radius min}
    B=ScalarParameter() {defaultValue=0.5, group=BrickSoup, parameterName=Core Glow_Fragments radius max}
    Alpha=Saturate(None=@n142)
   )
  )
  n144 = ScalarParameter() {defaultValue=-0.2, group=BrickSoup, parameterName=Core Glow_Fragments Bias}
  n145 = NamedRerouteUsage()
  n146 = LinearInterpolate(
   A=Multiply(A=Add(A=@n143, B=@n144), B=ScalarParameter() {defaultValue=1.5, group=BrickSoup, parameterName=Core Glow_Fragments Mult})
   B=LinearInterpolate(
    A=ScalarParameter() {defaultValue=0.3, group=BrickSoup, parameterName=Cracks_Min}
    B=ScalarParameter() {defaultValue=3, group=BrickSoup, parameterName=Cracks_Max}
    Alpha=Multiply(A=@n145, B=@n145)
   )
   Alpha=NamedRerouteUsage()
  )
  n147 = Add(A=NamedRerouteUsage(), B=ScalarParameter() {defaultValue=-0.8, group=BrickSoup, parameterName=Core Glow_Fresnel Bias})
  n148 = ScalarParameter() {defaultValue=0.3, group=BrickSoup, parameterName=Core Glow_Fresnel Mult}
  n149 = Saturate(None=Add(A=Add(A=@n141, B=@n146), B=Multiply(A=@n147, B=@n148)))
  n150 = NamedRerouteUsage()
  n151 = NamedRerouteDeclaration(None=Saturate(None=ComponentMask(None=@n150) {b=true})) {name=M_N_Top, nodeColor=<rgba 0 1 0.1797 1>}
  n152 = StaticBoolParameter() {group=BrickSoup, parameterName=Is_Fragment, sortPriority=0}
  NamedRerouteDeclaration(None=StaticSwitch(True=Multiply(A=@n149, B=@n151), False=@n149, Value=NamedRerouteDeclaration(None=@n152) {name=IsFragment, nodeColor=<rgba 1 0 0.1875 1>})) {name=M_InnerGlow, nodeColor=<rgba 0 0.5312 1 1>}
  n153 = ScalarParameter() {defaultValue=0.01, group=BrickSoup, parameterName=Cracks_Silhouete Min}
  n154 = ScalarParameter() {defaultValue=0.23, group=BrickSoup, parameterName=Cracks_Silhouete Max}
  n155 = ScalarParameter() {defaultValue=0.01, group=BrickSoup, parameterName=Cracks_Silhouete Min Mobile}
  n156 = ScalarParameter() {defaultValue=0.23, group=BrickSoup, parameterName=Cracks_Silhouete Max Mobile}
  NamedRerouteDeclaration(None=ShadingPathSwitch(
    Default=SmoothStep(Min=@n153, Max=@n154, Value=@n92) {constMax=0.05, constMin=0.01}
    Mobile=SmoothStep(Min=@n155, Max=@n156, Value=@n92) {constMax=0.05, constMin=0.01}
   )) {name=M_Cracks_Silhouette, nodeColor=<rgba 0.8516 1 0 1>}
  n157 = OneMinus(None=@n71)
  Saturate(None=LinearInterpolate(
    A=Constant3Vector() {constant=<rgba 0 1 0 0>}
    B=@n48
    Alpha=NamedRerouteDeclaration(None=@n157) {name=Mask_ProjetionBroken, nodeColor=<rgba 0 0.1562 1 1>}
   ))
 ]}
