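# Subsurface ice and snow material for BrickSoup meshes, supporting Nanite with normal-map fallbacks for non-Nanite meshes.
# MF_Ice_2022 gets snow from a slope and distance-field-AO mask, which also drives Ice/Snow physical material output.
# Height- and normal-based blending merges it into landscape runtime virtual textures, with waterline darkening near world zero.
# Switches add snow-mesh mode and time-of-day glow; destruction damage, ice cracks and particle-color emissive finish it.
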Material /BrickSoup/Assets/Materials/M_BrickSoup_ICE
graph {
 properties {
  bAutomaticallySetUsageInEditor=false
  bTangentSpaceNormal=false
  bUseMaterialAttributes=true
  bUsedWithInstancedStaticMeshes=true
  bUsedWithNanite=true
  bUsedWithNiagaraMeshParticles=true
  bUsedWithSplineMeshes=true
  physMaterial=/Game/FortressPhysicalMaterials/Stone
  pixelDepthOffsetMode=PDOM_Legacy
  shadingModel=MSM_Subsurface
  usedShadingModels=Subsurface
 }
 nodes [
  n1 = ParticleColor()
  n2 = Ceil(None=ParticleRelativeTime())
  n3 = NamedRerouteUsage()
  n4 = ScalarParameter() {group=A_Emissive_TOD, parameterName=DayMult, sortPriority=-3}
  n5 = CollectionParameter() {
   collection=&FortniteMaterialParameters
   parameterName=FortniteActiveTimeOfDay
  }
  n6 = DotProduct(
   A=MaterialFunctionCall(X=@n4, Y=@n4, Z=@n4, A=ScalarParameter() {defaultValue=1, group=A_Emissive_TOD, parameterName=NightMult, sortPriority=-2}) {materialFunction=&MakeFloat4}.Result
   B=@n5
  )
  n7 = WorldPosition()
  n8 = ActorPositionWS()
  n9 = Divide(A=ComponentMask(None=Subtract(A=@n7.XYZ, B=@n8)) {b=true}, B=Constant() {r=1000})
  n10 = Max(A=ComponentMask(None=VertexNormalWS()) {b=true}) {constB=0.5}
  n11 = Multiply(A=VectorParameter() {parameterName=NightGlow, defaultValue=<rgba 1 0.1285 0 1>}.RGB, B=Saturate(None=Multiply(A=@n9, B=@n10)))
  n12 = NamedRerouteUsage()
  n13 = Saturate(None=Max(A=NamedRerouteUsage(), B=OneMinus(None=@n12)))
  n14 = NamedRerouteUsage()
  n15 = NamedRerouteUsage()
  n16 = ComponentMask(None=VertexNormalWS()) {b=true}
  n17 = OneMinus(None=@n16)
  n18 = Multiply(A=@n17, B=Abs(None=@n16))
  n19 = BlendMaterialAttributes(
   A=@n15
   B=@n14
   Alpha=StaticSwitchParameter(True=Saturate(None=@n17), False=Saturate(None=@n18)) {group=RVT, parameterName="Skip RVT on Bottom && Sides"}
  )
  n20 = StaticSwitchParameter(
   True=@n14
   False=BlendMaterialAttributes(A=@n14, B=StaticSwitchParameter(True=@n19, False=@n15) {group=RVT, parameterName=Skip RVT on Bottom}, Alpha=NamedRerouteUsage())
  ) {group=RVT, parameterName=SkipRVT}
  n21 = VertexColor()
  n22 = BlendMaterialAttributes(
   A=@n20
   B=MaterialFunctionCall!(Mask=@n13, SlipperySnow_Mask, "Material;"=@n20) {materialFunction=/Game/Athena/Asteria/Environments/Landscape/Materials/MaterialFunctions/MF_SnowPostRVT}.SnowMaterial
   Alpha=StaticSwitchParameter(True=StaticSwitchParameter(True=@n21.R, False=@n13) {group=1_SnowMeshes, parameterName=Snow Meshes Vertex Mask, sortPriority=-1}, False=@n13) {group=1_SnowMeshes, parameterName=Snow Meshes, sortPriority=-1}
  )
  n23 = GetMaterialAttributes(None=@n22) {attributeGetTypes [#1, #2]}
  n24 = Constant3Vector() {constant=<rgba 0.1197 0.1364 0.1615 0.2>}
  n25 = BlendMaterialAttributes(
   A=SetMaterialAttributes!(MaterialAttributes=@n23.MaterialAttributes, Base Color=Multiply(A=@n23.Base Color, B=@n24), Roughness=Multiply(A=@n23.Roughness, B=@n24)) {attributeSetTypes [#1, #2]}
   B=@n22
   Alpha=NamedRerouteUsage()
  )
  n26 = StaticSwitchParameter(
   True=SetMaterialAttributes!(MaterialAttributes=@n25, Emissive Color=Multiply(A=@n6, B=@n11)) {attributeSetTypes [#3]}
   False=@n25
  ) {group=1_SnowMeshes, parameterName=Night Glow, sortPriority=-1}
  n27 = MaterialFunctionCall(
   NoRVT=@n3
   RVT=MaterialProxyReplace(Realtime=StaticSwitchParameter(True=@n26, False=@n25) {group=1_SnowMeshes, parameterName=Snow Meshes, sortPriority=-1}, MaterialProxy=@n3)
  ) {materialFunction=&FN_RVT_Switch_Material}
  n28 = ShadingPathSwitch(Default=@n27.Result, Deferred=@n27.Result, Forward=@n3, Mobile=@n3)
  n29 = StaticSwitchParameter(True=StaticBool() {value=true}, False=StaticBool()) {parameterName=Use Additive WPO}
  n30 = GetMaterialAttributes(None=StaticSwitchParameter(
    True=MaterialFunctionCall(Material_In=@n28, UseAdditiveWPO=@n29) {materialFunction=&MF_FN_DestructibleDamage}.Material_Out
    False=@n28
   ) {defaultValue=true, group=Damage, parameterName=DestructionDamage}) {attributeGetTypes [
    #3
    #1
    #4
    #5
    #6
   ]}
  n31 = GetMaterialAttributes(None=MaterialFunctionCall!(In=@n30.MaterialAttributes) {materialFunction=&/MF_IceCracks}.MA) {attributeGetTypes [
    #3
    #4
    #5
    #6
   ]}
  n32 = SetMaterialAttributes!(
   MaterialAttributes=@n30.MaterialAttributes
   Emissive Color=Add(A=LinearInterpolate(B=@n1.RGB, Alpha=@n2), B=Add(A=@n30.Emissive Color, B=@n31.Emissive Color))
   Base Color
   Subsurface Color=@n31.Subsurface Color
   Opacity=Add(A=@n30.Opacity, B=@n31.Opacity)
   Normal=@n31.Normal
  ) {attributeSetTypes [
    #3
    #1
    #4
    #5
    #6
   ]}
  n33 = BlendMaterialAttributes(A=@n32, B=@n30.MaterialAttributes, Alpha=NamedRerouteUsage())
  n34 = GetMaterialAttributes(None=NamedRerouteUsage()) {attributeGetTypes [#1, #4]}
  MaterialProxyReplace(
   Realtime=@n33
   MaterialProxy=SetMaterialAttributes!(
    MaterialAttributes=GetMaterialAttributes(None=@n33) {attributeGetTypes [#4, #1, #3]}.MaterialAttributes
    Base Color=Add(A=@n34.Base Color, B=@n34.Base Color)
   ) {attributeSetTypes [#1]}
  )
  n35 = MaterialFunctionCall!(RVT=NamedRerouteUsage()) {materialFunction=/Game/Environments/Asteria/Rocks/Materials/MaterialFunctions/MF_Ice_2022}
  NamedRerouteDeclaration(None=@n35.Result) {name=tes, nodeColor=<rgba 0.6172 1 0 1>}
  n36 = NamedRerouteUsage()
  n37 = VertexNormalWS()
  n38 = RuntimeVirtualTextureSampleParameter!(UVs, World Position, Mip Bias=LinearInterpolate(Alpha=ComponentMask(None=@n37) {b=true}) {constA=6, constB=0}) {
   group=RVT
   materialType=BaseColor_Normal_Specular_YCoCg
   mipValueMode=RVTMVM_MipBias
   parameterName=RVT
   virtualTexture=/Game/Athena/Apollo/Maps/Landscape/VT/RVT_Apollo_Landscape
  }
  n39 = WorldPosition()
  n40 = RuntimeVirtualTextureSampleParameter!(UVs, World Position) {
   group=RVT
   materialType=WorldHeight
   parameterName=RVTHeight
   virtualTexture=/Game/Athena/Apollo/Maps/Landscape/VT/RVT_Apollo_Landscape_Height
  }
  n41 = Subtract(
   A=Add(A=ScalarParameter() {group=RVT, parameterName=Offset RVT WP}, B=ComponentMask(None=@n39.XYZ) {b=true})
   B=NamedRerouteDeclaration(None=@n40.WorldHeight) {name=RVT Height as is, nodeColor=<rgba 0 0.4141 1 1>}
  )
  n42 = NamedRerouteDeclaration(None=@n41) {name=RVT Height, nodeColor=<rgba 1 0 0.2344 1>}
  n43 = ScalarParameter() {
   defaultValue=25
   desc="Controls the blending of normals on sides of meshes, but not UP vector surfaces"
   group=RVT
   parameterName=Blend RVT Normals
  }
  n44 = ScalarParameter() {
   defaultValue=300
   desc="Used to control how far colors will blend with the rvt, as well sets blending distance of normals for Up looking surfaces"
   group=RVT
   parameterName=Blend RVT
  }
  n45 = Saturate(None=ComponentMask(None=VertexNormalWS()) {b=true})
  n46 = LinearInterpolate(A=@n43, B=@n44, Alpha=@n45)
  n47 = BlendMaterialAttributes(
   A=SetMaterialAttributes!(MaterialAttributes=@n36, Normal=NamedRerouteDeclaration(None=@n38.Normal) {name=RVT N, nodeColor=<rgba 0 0.6953 1 1>}) {attributeSetTypes [#6]}
   B=@n36
   Alpha=Saturate(None=Divide(A=@n42, B=@n46))
  ) {vertexAttributeBlendType=UseA}
  n48 = SetMaterialAttributes!(
   MaterialAttributes=@n47
   Base Color=NamedRerouteDeclaration(None=@n38.BaseColor) {name=RVT Base Color, nodeColor=<rgba 0 1 0.6016 1>}
   Specular=NamedRerouteDeclaration(None=@n38.Specular) {name=RVT Specular, nodeColor=<rgba 1 0 0.8203 1>}
   Roughness=NamedRerouteDeclaration(None=@n38.Roughness) {name=RVT Rough, nodeColor=<rgba 0 1 0.7422 1>}
   Subsurface Color
  ) {attributeSetTypes [
    #1
    #7
    #2
    #4
   ]}
  NamedRerouteDeclaration(None=BlendMaterialAttributes(A=NamedRerouteDeclaration(None=@n48) {name=RVT Colors, nodeColor=<rgba 0.5938 1 0 1>}, B=@n47, Alpha=NamedRerouteUsage()) {vertexAttributeBlendType=UseA}) {name=RVT On, nodeColor=<rgba 0 0.8594 1 1>}
  n49 = LinearInterpolate(A=@n43, B=@n44, Alpha=@n45)
  NamedRerouteDeclaration(None=MaterialFunctionCall(Transition Phase=Divide(A=@n41, B=@n49), Height Texture=NamedRerouteUsage()) {materialFunction=&HeightLerp}.Lerp Alpha No Contrast) {name=RVT Blending, nodeColor=<rgba 0.9922 1 0 1>}
  n50 = NamedRerouteDeclaration(None=@n35.Result) {name=Ice As Is, nodeColor=<rgba 0 0.0859 1 1>}
  n51 = GetMaterialAttributes(None=MaterialFunctionCall() {materialFunction=/Game/Athena/Asteria/Environments/Landscape/Materials/MaterialFunctions/MF_AsteriaLandscapeLayers}.Snow) {attributeGetTypes [#1, #5, #6]}
  n52 = TextureObjectParameter() {
   group=1_SnowMeshes
   parameterName=Snow Normals
   samplerType=SAMPLERTYPE_Normal
   texture=/Game/Environments/Asteria/Terrain/Textures/T_Asteria_WindBlownSnow_02_N
  }
  n53 = ScalarParameter() {defaultValue=1, group=1_SnowMeshes, parameterName=Snow Meshes UV scale}
  n54 = Multiply(A=@n53) {constB=1024}
  n55 = SetMaterialAttributes!(
   MaterialAttributes=@n50
   Base Color=NamedRerouteDeclaration(None=@n51.Base Color) {name=LSnow_Base, nodeColor=<rgba 0.8984 0 1 1>}
   Normal=Multiply(
    A=Constant3Vector() {constant=<rgba 0.05 0.05 1 0>}
    B=MaterialFunctionCall(TextureObject=@n52, TextureSize=@n54) {materialFunction=&WorldAlignedNormal}.Z Texture
   )
  ) {attributeSetTypes [#1, #6]}
  n56 = NamedRerouteDeclaration(None=@n35.Noise) {name=Noise, nodeColor=<rgba 0.5703 1 0 1>}
  n57 = NamedRerouteDeclaration(None=MaterialFunctionCall(Transition Phase=NamedRerouteUsage(), Height Texture=OneMinus(None=@n56)) {materialFunction=&HeightLerp}.Lerp Alpha No Contrast) {name=Snow Mask 2, nodeColor=<rgba 1 0 0.3984 1>}
  n58 = GetMaterialAttributes(None=StaticSwitchParameter(True=NamedRerouteUsage(), False=BlendMaterialAttributes(A=@n50, B=@n55, Alpha=@n57)) {group=1_SnowMeshes, parameterName=Snow Meshes, sortPriority=-1}) {attributeGetTypes [#6]}
  n59 = NamedRerouteDeclaration(None=@n58.Normal) {name=N Tile, nodeColor=<rgba 0 0.0859 1 1>}
  n60 = TextureCoordinate() {coordinateIndex=2}
  n61 = TextureSampleParameter2D(UVs=@n60) {
   desc="Use to support Non Nanite meshes fallbacks. "
   group=Ice Fallback
   parameterName=Fallback Normal
   samplerType=SAMPLERTYPE_Normal
   texture=&FlatNormalBlue
  }
  n62 = TextureSampleParameter2D() {
   desc="Use to support Non Nanite meshes fallbacks. "
   group=Ice Fallback
   parameterName=Fallback Normal
   samplerType=SAMPLERTYPE_Normal
   texture=&FlatNormalBlue
  }
  n63 = StaticSwitchParameter(
   True=MaterialFunctionCall!(UVs=@n60, Normal=@n61.RGB) {materialFunction=/Engine/Functions/Engine_MaterialFunctions01/Texturing/DeriveTangentBasis}.World Space Normal
   False=@n62.RGB
  ) {group=Ice Fallback, parameterName=Use UV2 for Fallback}
  n64 = StaticSwitchParameter(True=Transform(None=@n63) {transformSourceType=TRANSFORMSOURCE_World, transformType=TRANSFORM_Tangent}, False=@n63) {group=Ice Fallback, parameterName=Use UV2 for Fallback}
  n65 = Constant3Vector() {constant=<rgba 1 1 0 0>}
  n66 = QualitySwitch(
   Default=Normalize(VectorInput=MaterialFunctionCall(BaseNormal=@n59, AdditionalNormal=@n64) {materialFunction=&BlendAngleCorrectedNormals}.Result)
   Low=Add(A=@n58.Normal, B=Multiply(A=@n64, B=@n65))
  )
  n67 = GetMaterialAttributes(None=SetMaterialAttributes!(MaterialAttributes=@n58.MaterialAttributes, Normal=NaniteReplace!(Default=@n66, Nanite=@n58.Normal)) {attributeSetTypes [#6]}) {attributeGetTypes [#6]}
  n68 = Transform(None=@n67.Normal)
  NamedRerouteDeclaration(None=SetMaterialAttributes(
    MaterialAttributes=@n67.MaterialAttributes
    Normal=NamedRerouteDeclaration(None=@n68) {name=Material Normals WS, nodeColor=<rgba 0.7344 1 0 1>}
   ) {attributeSetTypes [#6]}) {name=PreRVT, nodeColor=<rgba 0.4766 1 0 1>}
  n69 = Multiply(A=TextureCoordinate(), B=@n53)
  n70 = TextureSampleParameter2D(UVs=@n69) {
   group=1_SnowMeshes
   parameterName=Snow Spec
   samplerType=SAMPLERTYPE_Masks
   texture=/Game/Environments/Asteria/Terrain/Textures/T_Asteria_WindBlownSnow_02_S
  }
  n71 = TextureSampleParameter2D(UVs=@n69) {
   group=1_SnowMeshes
   parameterName=Snow Normals
   samplerType=SAMPLERTYPE_Normal
   texture=&T_Asteria_WindBlownSnow_02_N
  }
  n72 = SetMaterialAttributes!(
   MaterialAttributes
   Base Color=NamedRerouteUsage()
   Specular=@n70.R
   Roughness=@n70.R
   Normal=@n71.RGB
   Subsurface Color=VectorParameter() {group=1_SnowMeshes, parameterName=SSS Tint, defaultValue=<rgba 0.0546 0.0885 0.1615 0>}.RGB
   Opacity=NamedRerouteUsage()
  ) {attributeSetTypes [
    #1
    #7
    #2
    #6
    #4
    #5
   ]}
  NamedRerouteDeclaration(None=@n72) {name=Snow Meshes, nodeColor=<rgba 1 0.7969 0 1>}
  NamedRerouteDeclaration(None=@n51.Opacity) {name=SnowSSS, nodeColor=<rgba 1 0.8672 0 1>}
  n73 = ComponentMask(None=WorldPosition().XYZ) {b=true}
  n74 = Saturate(None=ComponentMask(None=VertexNormalWS()) {b=true})
  n75 = Add(A=@n73, B=LinearInterpolate(A=Constant() {r=155}, B=Constant() {r=-155}, Alpha=@n74))
  n76 = LinearInterpolate(A=Constant() {r=55}, B=Constant() {r=355}, Alpha=@n74)
  NamedRerouteDeclaration(None=Saturate(None=Divide(A=@n75, B=@n76))) {name=Ocean, nodeColor=<rgba 0.6406 1 0 1>}
  n77 = Add(A=Constant() {r=25}, B=@n73)
  n78 = Constant() {r=155}
  NamedRerouteDeclaration(None=Saturate(None=Divide(A=@n77, B=@n78))) {name=Ocean_Line, nodeColor=<rgba 0.6406 1 0 1>}
  n79 = TextureSampleParameter2D(UVs=StaticSwitchParameter(True=TextureCoordinate() {coordinateIndex=2}, False=TextureCoordinate()) {group=Ice Fallback, parameterName=Use UV2 for Fallback}) {
   desc="Use to support Non Nanite meshes fallbacks. "
   group=Ice Fallback
   parameterName=Fallback Normal
   samplerType=SAMPLERTYPE_Normal
   texture=&FlatNormalBlue
  }
  n80 = SphereMask(
   A=NaniteReplace(Default=Transform(None=@n79.RGB), Nanite=VertexNormalWS())
   B=Constant3Vector() {constant=<rgba 0 0 1 1>}
   Radius=ScalarParameter() {defaultValue=0.9, group=2_Sides/Up, parameterName=Ice Up/Sides}
   Hardness=ScalarParameter() {defaultValue=0.35, group=2_Sides/Up, parameterName=Ice Up/Sides Contrast}
  )
  n81 = Constant() {r=25}
  n82 = Constant() {r=250}
  n83 = OneMinus(None=Multiply(A=VertexColor().G, B=DistanceFieldApproxAO!(World Position, Normal, BaseDistance=@n81, Radius=@n82) {numSteps=3}))
  n84 = ScalarParameter() {defaultValue=0.7, parameterName=Snow In Cracks}
  n85 = NaniteReplace(Default=@n80, Nanite=Max(A=@n80, B=Multiply(A=@n83, B=@n84)))
  n86 = NamedRerouteDeclaration(None=Saturate(None=FeatureLevelSwitch(Default=@n85, ES3_1=@n80))) {name=SnowMask, nodeColor=<rgba 0 0.2266 1 1>}
  MaterialExpressionLinearInterpolate_17 = LinearInterpolate(Alpha=@n86) {constA=1, constB=0}
  MaterialExpressionLinearInterpolate_16 = LinearInterpolate(Alpha=@n86)
  PhysicalMaterialOutput!(Ice=@MaterialExpressionLinearInterpolate_17, Snow=@MaterialExpressionLinearInterpolate_16) {inputs [
    {
     physicalMaterial=/Game/FortressPhysicalMaterials/Ice
     input {
      expression=&self:MaterialExpressionLinearInterpolate_17
      inputName=None
      mask=0
      maskA=0
      maskB=0
      maskG=0
      maskR=0
      outputIndex=0
     }
    }
    {
     physicalMaterial=/Game/FortressPhysicalMaterials/Snow
     input {
      expression=&self:MaterialExpressionLinearInterpolate_16
      inputName=None
      mask=0
      maskA=0
      maskB=0
      maskG=0
      maskR=0
      outputIndex=0
     }
    }
   ]}
 ]
}
